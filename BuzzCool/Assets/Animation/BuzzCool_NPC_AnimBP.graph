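# Animation Blueprint for the BuzzCool NPC, driving locomotion, dancing, melee and death animations.
# Each update derives movement flags from a speed value and copies attack, dance and death flags from the FortAIPawn.
# These drive a state machine with random idle, attack and death clips, plus a speed-driven running blendspace.
# Its output passes through DefaultSlot; the upper-body additive has zero alpha, and ActiveIdle is always set false.
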AnimBlueprint /BuzzCool/Assets/Animation/BuzzCool_NPC_AnimBP
target_skeleton=/BuzzCool/NPCs/BuzzCool_Clones/Meshes/M_MED_BuzzCool_NPC_Skeleton
graphs [
 AnimGraph {nodes [
   Output Pose(Result=Apply Additive(
     Base=Use cached pose 'LocoCache'()
     Additive="Sequence Player (additive) 'BuzzCool_UpperBody_FromStanding_InFront_ToBack'"() {bLoopAnimation=false, sequence=/BuzzCool/Assets/Animation/Locomotion/BuzzCool_UpperBody_FromStanding_InFront_ToBack}
     Alpha=0
    ))
   AnimGraphNode_StateMachine_0 = Locomotion() {is=title AnimGraphNode_StateMachine}
   Save cached pose 'LocoCache'(Pose=Slot 'DefaultSlot'(Source=@AnimGraphNode_StateMachine_0)) {cacheName=LocoCache}
   Sequence Player 'BuzzCool_Sprint'!(PlayRate:float=Get Playrate() -> (Playrate:double)) -> (Pose) {bLoopAnimation=false, sequence=/BuzzCool/Assets/Animation/Locomotion/BuzzCool_Sprint}
  ]}
 Locomotion {
  entry=Idles
  path=AnimGraph.AnimGraphNode_StateMachine_0.Locomotion
  ids=id
  nodes [
   Alias() {is=title AnimStateAliasNode, also {aliases [Idles,Running,Walking]}}
   IntoDeath() {is=title AnimStateAliasNode}
   ToIdle() {is=title AnimStateAliasNode, also {aliases [Running,Walking]}}
   AnimStateNode_0 = Idles() {is=title AnimStateNode}
   AnimStateNode_1 = Walking() {is=title AnimStateNode}
   AnimStateNode_2 = Running() {is=title AnimStateNode}
   AnimStateNode_3 = Attacking() {is=title AnimStateNode}
   AnimStateNode_4 = Death() {is=title AnimStateNode}
   AnimStateNode_5 = Dead() {is=title AnimStateNode}
   AnimStateTransitionNode_0 = AnimStateTransitionNode() {bAllowInertializationForSelfTransitions=true}
   AnimStateTransitionNode_1 = AnimStateTransitionNode() {bAllowInertializationForSelfTransitions=true}
   AnimStateTransitionNode_2 = AnimStateTransitionNode() {bAllowInertializationForSelfTransitions=true}
   AnimStateTransitionNode_3 = AnimStateTransitionNode() {bAllowInertializationForSelfTransitions=true}
   AnimStateTransitionNode_4 = AnimStateTransitionNode() {bAllowInertializationForSelfTransitions=true, bAutomaticRuleBasedOnSequencePlayerInState=true}
   AnimStateTransitionNode_5 = AnimStateTransitionNode() {bAllowInertializationForSelfTransitions=true}
   AnimStateTransitionNode_6 = AnimStateTransitionNode() {bAllowInertializationForSelfTransitions=true}
   AnimStateTransitionNode_7 = AnimStateTransitionNode() {bAllowInertializationForSelfTransitions=true, bAutomaticRuleBasedOnSequencePlayerInState=true}
  ]
  links [
   {
    dst=Attacking
    rule=AnimGraph.AnimGraphNode_StateMachine_0.Locomotion.AnimStateTransitionNode_3.Transition
    src=Alias
    via=AnimStateTransitionNode_3
   }
   {
    dst=Idles
    rule=AnimGraph.AnimGraphNode_StateMachine_0.Locomotion.AnimStateTransitionNode_4.Transition
    src=Attacking
    via=AnimStateTransitionNode_4
   }
   {
    dst=Dead
    rule=AnimGraph.AnimGraphNode_StateMachine_0.Locomotion.AnimStateTransitionNode_7.Transition
    src=Death
    via=AnimStateTransitionNode_7
   }
   {
    dst=Walking
    rule=AnimGraph.AnimGraphNode_StateMachine_0.Locomotion.AnimStateTransitionNode_0.Transition
    src=Idles
    via=AnimStateTransitionNode_0
   }
   {
    dst=Death
    rule=AnimGraph.AnimGraphNode_StateMachine_0.Locomotion.AnimStateTransitionNode_6.Transition
    src=IntoDeath
    via=AnimStateTransitionNode_6
   }
   {
    dst=Walking
    rule=AnimGraph.AnimGraphNode_StateMachine_0.Locomotion.AnimStateTransitionNode_5.Transition
    src=Running
    via=AnimStateTransitionNode_5
   }
   {
    dst=Idles
    rule=AnimGraph.AnimGraphNode_StateMachine_0.Locomotion.AnimStateTransitionNode_2.Transition
    src=ToIdle
    via=AnimStateTransitionNode_2
   }
   {
    dst=Running
    rule=AnimGraph.AnimGraphNode_StateMachine_0.Locomotion.AnimStateTransitionNode_1.Transition
    src=Walking
    via=AnimStateTransitionNode_1
   }
  ]
 }
 Idles {
  path=AnimGraph.AnimGraphNode_StateMachine_0.Locomotion.AnimStateNode_0.Idles
  nodes [
   n1 = Random Sequence Player!() -> (Pose:Pose Link Structure) {
    is=title AnimGraphNode_RandomPlayer
    entries [
     {
      chanceToPlay=1
      maxLoopCount=0
      maxPlayRate=1
      minLoopCount=0
      minPlayRate=1
      sequence=/BuzzCool/Assets/Animation/Locomotion/BuzzCool_Idles_LookingAround
      blendIn {blendOption=HermiteCubic, blendTime=0.2, customCurve=null}
     }
     {
      chanceToPlay=1
      maxLoopCount=0
      maxPlayRate=1
      minLoopCount=0
      minPlayRate=1
      sequence=/BuzzCool/Assets/Animation/Locomotion/BuzzCool_Idles_ButtScratch
      blendIn {blendOption=HermiteCubic, blendTime=0.2, customCurve=null}
     }
    ]
   }
   n2 = Random Sequence Player() {entries [
     {
      chanceToPlay=1
      maxLoopCount=0
      maxPlayRate=1
      minLoopCount=0
      minPlayRate=1
      sequence=/BuzzCool/Assets/Animation/Locomotion/BuzzCool_Idles
      blendIn {blendOption=HermiteCubic, blendTime=0.2, customCurve=null}
     }
     {
      chanceToPlay=1
      maxLoopCount=0
      maxPlayRate=1
      minLoopCount=0
      minPlayRate=1
      sequence=&BuzzCool_Idles_ButtScratch
      blendIn {blendOption=HermiteCubic, blendTime=0.2, customCurve=null}
     }
     {
      chanceToPlay=1
      maxLoopCount=0
      maxPlayRate=1
      minLoopCount=0
      minPlayRate=1
      sequence=/BuzzCool/Assets/Animation/Locomotion/BuzzCool_Idles_DonutFingers
      blendIn {blendOption=HermiteCubic, blendTime=0.2, customCurve=null}
     }
     {
      chanceToPlay=1
      maxLoopCount=0
      maxPlayRate=1
      minLoopCount=0
      minPlayRate=1
      sequence=/BuzzCool/Assets/Animation/Locomotion/BuzzCool_Idles_DonutFists
      blendIn {blendOption=HermiteCubic, blendTime=0.2, customCurve=null}
     }
     {
      chanceToPlay=1
      maxLoopCount=0
      maxPlayRate=1
      minLoopCount=0
      minPlayRate=1
      sequence=/BuzzCool/Assets/Animation/Locomotion/BuzzCool_Idles_DonutTongue
      blendIn {blendOption=HermiteCubic, blendTime=0.2, customCurve=null}
     }
     {
      chanceToPlay=1
      maxLoopCount=0
      maxPlayRate=1
      minLoopCount=0
      minPlayRate=1
      sequence=/BuzzCool/Assets/Animation/Locomotion/BuzzCool_Idles_IdleLoop
      blendIn {blendOption=HermiteCubic, blendTime=0.2, customCurve=null}
     }
     {
      chanceToPlay=1
      maxLoopCount=0
      maxPlayRate=1
      minLoopCount=0
      minPlayRate=1
      sequence=&BuzzCool_Idles_LookingAround
      blendIn {blendOption=HermiteCubic, blendTime=0.2, customCurve=null}
     }
    ]}
   n3 = Get ActiveIdle() -> (ActiveIdle:bool)
   Output Animation Pose(Result=Blend Poses by bool(
     BlendPose_0=Sequence Player 'BuzzCool_DanceGrenade_Dance'() {sequence=/BuzzCool/Assets/Animation/Locomotion/BuzzCool_DanceGrenade_Dance}
     BlendPose_1=Blend Poses by bool(BlendPose_0=@n1, BlendPose_1=@n2, bActiveValue=@n3)
     bActiveValue=Get IsDancing() -> (IsDancing:bool)
    ))
  ]
 }
 Walking {
  path=AnimGraph.AnimGraphNode_StateMachine_0.Locomotion.AnimStateNode_1.Walking
  nodes [Output Animation Pose(Result=Sequence Player 'BuzzCool_Walks'() {bLoopAnimation=false, sequence=/BuzzCool/Assets/Animation/Locomotion/BuzzCool_Walks})]
 }
 Running {
  path=AnimGraph.AnimGraphNode_StateMachine_0.Locomotion.AnimStateNode_2.Running
  nodes [Output Animation Pose(Result=Blendspace Player 'BuzzCool_TwoArms_Follow_BS'!(X=0, Y=Get MoveSpeed() -> (MoveSpeed:double), PlayRate:float=1) -> (Pose) {blendSpace=/BuzzCool/Assets/Animation/Locomotion/BuzzCool_TwoArms_Follow_BS})]
 }
 Attacking {
  path=AnimGraph.AnimGraphNode_StateMachine_0.Locomotion.AnimStateNode_3.Attacking
  nodes [
   n1 = Random Sequence Player() {
    bShuffleMode=true
    entries [
     {
      chanceToPlay=1
      maxLoopCount=0
      maxPlayRate=1
      minLoopCount=0
      minPlayRate=1
      sequence=/BuzzCool/Assets/Animation/Locomotion/BuzzCool_Melee_AnimSet
      blendIn {blendOption=HermiteCubic, blendTime=0.2, customCurve=null}
     }
     {
      chanceToPlay=1
      maxLoopCount=0
      maxPlayRate=1
      minLoopCount=0
      minPlayRate=1
      sequence=&BuzzCool_Melee_AnimSet
      blendIn {blendOption=HermiteCubic, blendTime=0.2, customCurve=null}
     }
     {
      chanceToPlay=1
      maxLoopCount=0
      maxPlayRate=1
      minLoopCount=0
      minPlayRate=1
      sequence=/BuzzCool/Assets/Animation/Locomotion/BuzzCool_Melee_Kick
      blendIn {blendOption=HermiteCubic, blendTime=0.2, customCurve=null}
     }
    ]
   }
   Output Animation Pose(Result=@n1)
  ]
 }
 Death {
  path=AnimGraph.AnimGraphNode_StateMachine_0.Locomotion.AnimStateNode_4.Death
  nodes [
   n1 = Random Sequence Player() {entries [
     {
      chanceToPlay=1
      maxLoopCount=0
      maxPlayRate=1
      minLoopCount=0
      minPlayRate=1
      sequence=/BuzzCool/Assets/Animation/Locomotion/BuzzCool_Death_Doh1
      blendIn {blendOption=HermiteCubic, blendTime=0.2, customCurve=null}
     }
     {
      chanceToPlay=1
      maxLoopCount=0
      maxPlayRate=1
      minLoopCount=0
      minPlayRate=1
      sequence=/BuzzCool/Assets/Animation/Locomotion/BuzzCool_Death_Doh2
      blendIn {blendOption=HermiteCubic, blendTime=0.2, customCurve=null}
     }
     {
      chanceToPlay=1
      maxLoopCount=0
      maxPlayRate=1
      minLoopCount=0
      minPlayRate=1
      sequence=/BuzzCool/Assets/Animation/Locomotion/BuzzCool_Death_Doh3
      blendIn {blendOption=HermiteCubic, blendTime=0.2, customCurve=null}
     }
     {
      chanceToPlay=1
      maxLoopCount=0
      maxPlayRate=1
      minLoopCount=0
      minPlayRate=1
      sequence=/BuzzCool/Assets/Animation/Locomotion/BuzzCool_Death_Doh4
      blendIn {blendOption=HermiteCubic, blendTime=0.2, customCurve=null}
     }
    ]}
   Output Animation Pose(Result=@n1)
  ]
 }
 Dead {
  path=AnimGraph.AnimGraphNode_StateMachine_0.Locomotion.AnimStateNode_5.Dead
  nodes [Output Animation Pose(Result=Sequence Player 'BuzzCool_DiedPose'() {bLoopAnimation=false, sequence=/BuzzCool/Assets/Animation/Locomotion/BuzzCool_DiedPose})]
 }
 Transition {
  path=AnimGraph.AnimGraphNode_StateMachine_0.Locomotion.AnimStateTransitionNode_0.Transition
  nodes [Result(bCanEnterTransition=Get IsMoving() -> (IsMoving:bool))]
 }
 Transition {
  path=AnimGraph.AnimGraphNode_StateMachine_0.Locomotion.AnimStateTransitionNode_1.Transition
  nodes [Result(bCanEnterTransition=Get Running() -> (Running:bool))]
 }
 Transition {
  path=AnimGraph.AnimGraphNode_StateMachine_0.Locomotion.AnimStateTransitionNode_2.Transition
  nodes [Result(bCanEnterTransition=NOT Boolean(A=Get IsMoving()))]
 }
 Transition {
  path=AnimGraph.AnimGraphNode_StateMachine_0.Locomotion.AnimStateTransitionNode_3.Transition
  nodes [Result(bCanEnterTransition=Get Attack() -> (Attack:bool))]
 }
 Transition {path=AnimGraph.AnimGraphNode_StateMachine_0.Locomotion.AnimStateTransitionNode_4.Transition, nodes [Result()]}
 Transition {
  path=AnimGraph.AnimGraphNode_StateMachine_0.Locomotion.AnimStateTransitionNode_5.Transition
  nodes [
   n1 = Get Running()
   Result(bCanEnterTransition=AND Boolean(A=Get IsMoving(), B=NOT Boolean(A=@n1)))
  ]
 }
 Transition {
  path=AnimGraph.AnimGraphNode_StateMachine_0.Locomotion.AnimStateTransitionNode_6.Transition
  nodes [Result(bCanEnterTransition=Get Died() -> (Died:bool))]
 }
 Transition {path=AnimGraph.AnimGraphNode_StateMachine_0.Locomotion.AnimStateTransitionNode_7.Transition, nodes [Result()]}
 EventGraph {nodes [
   Event BlueprintUpdateAnimation()
   > n1 = Sequence()
   n2 = Cast To FortAIPawn!(execute=@n1.then_1, Object=TryGetPawnOwner()) -> (then, CastFailed, AsFort AIPawn:Fort AIPawn Object Reference) {is=title K2Node_DynamicCast}
   > Set Attack(Attack:bool=Get bAttacking!(self:Fort AIPawn Object Reference=@n2) -> (bAttacking:bool)) -> (Output_Get:bool)
   > Set IsDancing(IsDancing:bool=Get bDancing!(self:Fort AIPawn Object Reference=@n2) -> (bDancing:bool)) -> (Output_Get:bool)
   > Set Full Body Hit React(Full Body Hit React:bool=Get bFullBodyHitReact!(self:Fort AIPawn Object Reference=@n2) -> (bFullBodyHitReact:bool)) -> (Output_Get:bool)
   > Set Died(Died:bool=Get bDied!(self:Fort AIPawn Object Reference=@n2) -> (bDied:bool)) -> (Output_Get:bool)
   > Set Has Engaged(Has Engaged:bool=Get bHasEngaged!(self:Fort AIPawn Object Reference=@n2) -> (bHasEngaged:bool)) -> (Output_Get:bool)
   > Set ActiveIdle(ActiveIdle:bool=false) -> (Output_Get:bool)
   n3 = Set MoveSpeed(execute=@n1.then_0, MoveSpeed:double=Vector Length(A=Property Access!() -> (Value:Vector=~) {is=title K2Node_PropertyAccess})) -> (Output_Get:double)
   > Set Playrate(Playrate:double=float / float(A=@n3, B=550)) -> (Output_Get:double)
   > Set IsMoving(IsMoving:bool="float > float"(A=Get MoveSpeed(), B=1)) -> (Output_Get:bool)
   > Set Running(Running:bool="float > float"(A=Get MoveSpeed(), B=250)) -> (Output_Get:bool)
   GetOwningActor()
   Cast To Pawn_WildEstate_BuzzCool!(execute, Object:Wildcard) -> (then, CastFailed, AsPawn Wild Estate Buzz Cool:Pawn Wild Estate Buzz Cool Object Reference) {is=title K2Node_DynamicCast}
  ]}
]
variables [
 ActiveIdle {type=boolean}
 Attack {type=boolean}
 Died {type=boolean}
 Full Body Hit React {type=boolean}
 Has Engaged {type=boolean}
 IsDancing {type=boolean}
 IsMoving {type=boolean}
 MoveSpeed {type=number}
 Playrate {type=number}
 Running {type=boolean}
]
