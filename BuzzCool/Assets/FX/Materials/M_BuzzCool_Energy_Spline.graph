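# Translucent, two-sided unlit Niagara ribbon and mesh-particle material for a scrolling energy strand.
# Two distorted texture scrolls pass through a particle-driven panning mask, then thresholds split them into translucent, solid and core colors.
# Opacity fades by view angle and at both ends.
# World position offset combines SplineThicken, a camera-space sine wave and a pull toward a dynamic-parameter target, which feeds PosX into all three axes.
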Material /BuzzCool/Assets/FX/Materials/M_BuzzCool_Energy_Spline
graph {
 properties {
  bAutomaticallySetUsageInEditor=false
  bUseTranslucencyVertexFog=false
  bUsedWithNiagaraMeshParticles=true
  bUsedWithNiagaraRibbons=true
  blendMode=BLEND_Translucent
  pixelDepthOffsetMode=PDOM_Legacy
  shadingModel=MSM_Unlit
  twoSided=true
  usedShadingModels=Unlit
 }
 nodes [
  n1 = DynamicParameter() {desc=Index 2 parameter for Niagara, parameterIndex=1, defaultValue=<rgba 0 0 0 0>, paramNames [PosX,PosY,PosZ,Param4]}
  n2 = Constant() {r=1}
  n3 = Constant() {r=1}
  n4 = Multiply(A=Subtract(A=AppendVector(A=@n2, B=@n3)) {constB=0.5}, B=ScalarParameter() {defaultValue=30, parameterName=ActorPositionOffset}) {constB=10}
  n5 = Constant()
  n6 = Subtract(
   A=Add(
    A=MaterialFunctionCall(X=@n1.PosX, Y=@n1.PosX, Z=@n1.PosX) {materialFunction=&MakeFloat3}.Result
    B=AppendVector(A=@n4, B=@n5)
   )
   B=WorldPosition().XYZ
  )
  n7 = TextureCoordinate()
  n8 = Power(Base=OneMinus(None=ComponentMask(None=@n7) {g=true})) {constExponent=5}
  n9 = Multiply(A=TextureCoordinate(), B=ScalarParameter() {defaultValue=2.5, parameterName=WPO Tiling})
  n10 = Time()
  n11 = ScalarParameter() {defaultValue=0.5, parameterName=WPO Time Multi}
  n12 = Sine(None=Add(A=ComponentMask(None=@n9) {g=true}, B=Multiply(A=@n10, B=@n11) {constB=2}))
  n13 = Multiply(A=@n12, B=ScalarParameter() {defaultValue=20, group=WPO, parameterName=WPO_HeightX}) {constB=0.5}
  n14 = Multiply(A=@n12, B=ScalarParameter() {group=WPO, parameterName=WPO_HeightY}) {constB=0.5}
  n15 = ScalarParameter() {group=WPO, parameterName=WPO_HeightZ}
  n16 = Constant() {r=0.5}
  n17 = LinearInterpolate(
   A=Constant3Vector()
   B=AppendVector(A=AppendVector(A=@n13, B=@n14), B=Multiply(A=@n12, B=@n15) {constB=0.6})
   Alpha=SphereMask(
    A=MaterialFunctionCall() {materialFunction=&LinearGradient}.VGradient
    B=@n16
    Radius=@n16
   ) {hardnessPercent=0.5}
  )
  Add(
   A=Add(A=NamedRerouteUsage(), B=Multiply(A=@n6, B=@n8))
   B=Multiply(
    A=Transform(None=@n17) {transformSourceType=TRANSFORMSOURCE_Camera}
    B=ScalarParameter() {defaultValue=2, group=WPO, parameterName=WPO_Wave_Intenstiy}
   ) {constB=80}
  )
  n18 = VectorParameter() {parameterName=CoreColor, defaultValue=<rgba 1.7588 0.7396 2 1>}
  n19 = ScalarParameter() {defaultValue=0.02, parameterName=Color Threshold}
  n20 = Multiply(A=Time(), B=ComponentMask(None=VectorParameter() {parameterName=TextureScrollSpeed, defaultValue=<rgba 0 0.9 0 1>}.RGB) {g=true, r=true})
  n21 = Constant() {r=1}
  n22 = Add(
   A=ScalarParameter() {defaultValue=1.8, parameterName=Tiling}
   B=DynamicParameter() {defaultValue=<rgba 0.5 1 1 1>, paramNames [Mask Pan, "", Core Intensity, Offset Tiling]}.Offset Tiling
  )
  n23 = NamedRerouteDeclaration(None=Multiply(A=TextureCoordinate(), B=AppendVector(A=@n21, B=@n22))) {name=BaseUVs, nodeColor=<rgba 1 0.2578 0 1>}
  n24 = ScalarParameter() {group=Distortion, parameterName=Distortion Speed X, sliderMax=2, sliderMin=-2}
  n25 = ScalarParameter() {defaultValue=0.2, group=Distortion, parameterName=Distortion Speed Y, sliderMax=2, sliderMin=-2}
  n26 = Multiply(
   A=Panner(Coordinate=TextureCoordinate(), Speed=AppendVector(A=@n24, B=@n25)) {speedX=-1}
   B=AppendVector(
    A=ScalarParameter() {defaultValue=0.2, group=Distortion, parameterName=Distortion Tiling X, sliderMax=2}
    B=ScalarParameter() {defaultValue=0.3, group=Distortion, parameterName=Distortion Tiling Y, sliderMax=2}
   )
  ) {constB=2}
  n27 = Multiply(
   A=Subtract(A=TextureSampleParameter2D(UVs=@n26) {group=Distortion, parameterName=Distortion Texture, texture=&T_Inky_Smoke_Tile}.R) {constB=0.5}
   B=ScalarParameter() {defaultValue=0.2, group=Distortion, parameterName=Distortion Intensity, sliderMax=2, sliderMin=-2}
  )
  n28 = TextureSampleParameter2D(UVs=Add(A=@n27, B=Add(A=@n23, B=@n20))) {parameterName=Texture, texture=&/T_BuzCool_GasCloud_Ligthing_M}
  n29 = Multiply(A=Time(), B=ComponentMask(None=VectorParameter() {parameterName=TextureScrollSpeed2, defaultValue=<rgba 0 0.8 0 1>}.RGB) {g=true, r=true})
  n30 = TextureSampleParameter2D(UVs=Add(A=@n27, B=Add(A=@n23, B=@n29))) {parameterName=Texture, texture=&T_BuzCool_GasCloud_Ligthing_M}
  n31 = DynamicParameter() {defaultValue=<rgba 0.5 1 1 1>, paramNames [Mask Pan, "", Core Intensity, Offset Tiling]}
  n32 = Add(A=ComponentMask(None=TextureCoordinate()) {g=true}, B=LinearInterpolate(Alpha=@n31.Mask Pan) {constA=-1}) {constB=0}
  n33 = Multiply(A=Abs(None=Subtract(A=@n32) {constB=0.5})) {constB=3}
  n34 = Multiply(
   A=Power(Base=Add(A=@n28.R, B=@n30.G), Exp=ScalarParameter() {defaultValue=1.46, parameterName=Texture Power}) {constExponent=8}
   B=SmoothStep(Value=OneMinus(None=@n33)) {constMax=0.7, constMin=0.4}
  )
  n35 = Saturate(None=Round(None=Add(A=@n19, B=@n34)))
  n36 = LinearInterpolate(
   A=VectorParameter() {parameterName=Color_Translucent, defaultValue=<rgba 0.6038 0.0595 0.7084 1>}.RGB
   B=VectorParameter() {parameterName=Color, defaultValue=<rgba 0.9129 0.3301 0.9219 1>}.RGB
   Alpha=@n35
  )
  n37 = Multiply(A=Step(Y=ScalarParameter() {defaultValue=0.8, parameterName=CoreSize}, X=@n34) {constY=0.8}, B=@n31.Core Intensity)
  n38 = ParticleColor()
  MaterialFunctionCall(Emissive=Multiply(A=LinearInterpolate(A=@n36, B=@n18.RGB, Alpha=@n37), B=@n38.RGB)) {materialFunction=&MF_FX_EmissiveAdjustments}
  n39 = Subtract(A=@n34, B=ScalarParameter() {defaultValue=0.042, parameterName=Opacity Threshold})
  n40 = ScalarParameter() {defaultValue=0.512, parameterName=Opacity Threshold Range}
  n41 = Saturate(None=Add(A=@n35, B=Divide(A=@n39, B=@n40))) {desc=Opacity}
  n42 = Multiply(A=DotProduct(A=NamedRerouteUsage(), B=CameraVectorWS())) {constB=1.5}
  n43 = ScalarParameter() {parameterName=MaskTop}
  n44 = Add(A=@n43) {constB=0.25}
  n45 = ComponentMask(None=TextureCoordinate()) {g=true}
  n46 = OneMinus(None=@n45)
  n47 = ScalarParameter() {parameterName=MaskBottom}
  n48 = Add(A=@n47) {constB=0.4}
  Multiply(
   A=Multiply(A=Multiply(A=@n38.A, B=@n41), B=Saturate(None=@n42))
   B=Multiply(A=SmoothStep(Min=@n43, Max=@n44, Value=@n46) {constMax=0.4}, B=SmoothStep(Min=@n47, Max=@n48, Value=@n45) {constMax=0.2})
  )
  n49 = MaterialFunctionCall(
   WidthBase=ScalarParameter() {defaultValue=250, parameterName=WidthBase}
   WidthTip=ScalarParameter() {defaultValue=150, parameterName=WidthTip}
  ) {materialFunction=&SplineThicken}
  NamedRerouteDeclaration(None=@n49.WorldPositionOffset) {name=STWPO, nodeColor=<rgba 0 0.6719 1 1>}
  NamedRerouteDeclaration(None=@n49.Normal) {name=STN, nodeColor=<rgba 0.9922 1 0 1>}
 ]
}
