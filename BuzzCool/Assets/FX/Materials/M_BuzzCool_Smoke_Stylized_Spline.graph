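# Translucent unlit Niagara material for stylized, spline-thickened smoke trails with three-color tinting and an optional glow switch.
# Warped, pinched UVs plus panning noise distortion drive a channel-packed texture for color and opacity.
# Opacity combines scrolling dissolve, edge and depth fades, a dynamic-parameter pan mask and a view-angle fade.
# SplineThicken supplies width and normals, while a second dynamic parameter feeds a V-weighted position offset that bends one end.
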Material /BuzzCool/Assets/FX/Materials/M_BuzzCool_Smoke_Stylized_Spline
graph {
 properties {
  bUsedWithNiagaraMeshParticles=true
  bUsedWithNiagaraRibbons=true
  bUsedWithNiagaraSprites=true
  blendMode=BLEND_Translucent
  pixelDepthOffsetMode=PDOM_Legacy
  shadingModel=MSM_Unlit
  twoSided=true
  usedShadingModels=Unlit
 }
 nodes [
  n1 = DynamicParameter() {desc=Index 2 parameter for Niagara, parameterIndex=1, defaultValue=<rgba 0 0 0 0>, paramNames [PosX,PosY,PosZ,Param4]}
  n2 = Subtract(A=AppendVector(A=Constant() {r=1}, B=Constant() {r=1})) {constB=0.5}
  n3 = ScalarParameter() {defaultValue=30, parameterName=ActorPositionOffset}
  n4 = Add(
   A=MaterialFunctionCall(X=@n1.PosX, Y=@n1.PosX, Z=@n1.PosX) {materialFunction=&MakeFloat3}.Result
   B=AppendVector(A=Multiply(A=@n2, B=@n3) {constB=10}, B=Constant())
  )
  n5 = WorldPosition()
  n6 = OneMinus(None=ComponentMask(None=TextureCoordinate()) {g=true})
  Add(A=NamedRerouteUsage(), B=Multiply(A=Subtract(A=@n4, B=@n5.XYZ), B=Power(Base=@n6) {constExponent=5}))
  n7 = ParticleColor()
  n8 = TextureCoordinate()
  n9 = ComponentMask(None=MaterialFunctionCall(Texture Scale=ScalarParameter() {defaultValue=1, group=UV Manipulation, parameterName=UV End Pinch, sliderMax=1}) {materialFunction=&ScaleUVsByCenter}.UVs) {g=true}
  n10 = ComponentMask(None=@n8) {r=true}
  n11 = AppendVector(
   A=ComponentMask(None=@n8) {r=true}
   B=LinearInterpolate(A=ComponentMask(None=@n8) {g=true}, B=Saturate(None=@n9), Alpha=Power(Base=@n10) {constExponent=5})
  )
  n12 = ComponentMask(None=@n11) {g=true}
  n13 = MaterialFunctionCall(Texture Scale=ScalarParameter() {defaultValue=1, group=UV Manipulation, parameterName=UV V Puff}) {materialFunction=&ScaleUVsByCenter}
  n14 = AppendVector(
   A=Power(
    Base=ComponentMask(None=@n11) {r=true}
    Exp=ScalarParameter() {defaultValue=1, group=UV Manipulation, parameterName=UV U stretch, sliderMax=2}
   ) {constExponent=0.8}
   B=LinearInterpolate(A=@n12, B=ComponentMask(None=@n13.UVs) {g=true}, Alpha=Sine(None=@n12) {period=2})
  )
  n15 = ScalarParameter() {defaultValue=1, group=Distortion, parameterName=Distortion Speed X, sliderMax=2, sliderMin=-2}
  n16 = ScalarParameter() {group=Distortion, parameterName=Distortion Speed Y, sliderMax=2, sliderMin=-2}
  n17 = Multiply(
   A=Panner(Coordinate=TextureCoordinate(), Speed=AppendVector(A=@n15, B=@n16)) {speedX=-1}
   B=AppendVector(
    A=ScalarParameter() {defaultValue=1, group=Distortion, parameterName=Distortion Tiling X, sliderMax=2}
    B=ScalarParameter() {defaultValue=1, group=Distortion, parameterName=Distortion Tiling Y, sliderMax=2}
   )
  ) {constB=2}
  n18 = Multiply(
   A=Subtract(A=TextureSampleParameter2D(UVs=@n17) {group=Distortion, parameterName=Distortion Texture, texture=&T_Inky_Smoke_Tile}.R) {constB=0.5}
   B=ScalarParameter() {group=Distortion, parameterName=Distortion Intensity, sliderMax=2, sliderMin=-2}
  )
  n19 = AppendVector(
   A=ScalarParameter() {group=Main Texture, parameterName=Speed X, sliderMax=2, sliderMin=-2}
   B=ScalarParameter() {defaultValue=0.5, group=Main Texture, parameterName=Speed Y, sliderMax=2}
  )
  n20 = ScalarParameter() {defaultValue=1, group=Main Texture, parameterName=Tiling X, sliderMax=2}
  n21 = ScalarParameter() {defaultValue=1, group=Main Texture, parameterName=Tiling Y, sliderMax=2}
  n22 = Multiply(
   A=Panner(Coordinate=Add(A=@n14, B=@n18), Speed=Multiply(A=@n19) {constB=2}) {speedX=-1}
   B=AppendVector(A=Add(A=@n20) {constB=0}, B=Add(A=@n21) {constB=0})
  ) {constB=2}
  n23 = TextureSampleParameter2D(UVs=LinearInterpolate(
    A=@n22
    B=MaterialFunctionCall(XY=@n22) {materialFunction=&Swizzle}.YX
    Alpha=ScalarParameter() {defaultValue=1, parameterName=Use Swizzle}
   )) {desc=Channel packed texture, group=Main Texture, parameterName=Ribbon Texture, texture=/Game/Effects/Fort_Effects/Textures/T_PlantStand_Air}
  n24 = LinearInterpolate(
   A=LinearInterpolate(
    A=VectorParameter() {group=Color, parameterName=Color 1, defaultValue=<rgba 0.4809 0 0.5677 1>}.RGB
    B=VectorParameter() {group=Color, parameterName=Color 2, defaultValue=<rgba 0.2639 0 0.5 1>}.RGB
    Alpha=@n23.G
   )
   B=VectorParameter() {group=Color, parameterName=Color 3, defaultValue=<rgba 1 0.0978 0 1>}.RGB
   Alpha=@n23.B
  )
  MaterialFunctionCall(Emissive=StaticSwitchParameter(True=@n7.RGB, False=Multiply(A=@n24, B=@n7.RGB)) {group=Glow, parameterName=Glow}) {materialFunction=&MF_FX_EmissiveAdjustments}
  n25 = Max(A=@n23.R, B=@n23.R)
  n26 = LinearInterpolate(
   A=VectorParameter() {group=Color, parameterName=Color 1, defaultValue=<rgba 0.4809 0 0.5677 1>}.A
   B=VectorParameter() {group=Color, parameterName=Color 2, defaultValue=<rgba 0.2639 0 0.5 1>}.A
   Alpha=@n23.G
  )
  n27 = VectorParameter() {group=Color, parameterName=Color 3, defaultValue=<rgba 1 0.0978 0 1>}
  n28 = StaticSwitchParameter(
   True=Multiply(A=@n7.A, B=Subtract(A=@n23.A, B=@n23.A))
   False=Multiply(A=Max(A=@n25, B=@n23.B), B=LinearInterpolate(A=@n26, B=@n27.A, Alpha=@n23.B))
  ) {group=Glow, parameterName=Glow}
  n29 = ScalarParameter() {defaultValue=1, group=Dissolve, parameterName=Dissolve Value Min, sliderMax=-5}
  n30 = ScalarParameter() {defaultValue=1, group=Dissolve, parameterName=Dissolve Value Max, sliderMax=5}
  n31 = Panner(
   Coordinate=TextureCoordinate() {uTiling=2, vTiling=0.8}
   Speed=AppendVector(
    A=ScalarParameter() {defaultValue=1, group=Dissolve, parameterName=Dissolve Speed U, sliderMax=2}
    B=ScalarParameter() {defaultValue=1, group=Dissolve, parameterName=Dissolve Speed V, sliderMax=3}
   )
  ) {speedX=-1}
  n32 = AppendVector(
   A=ScalarParameter() {defaultValue=1, group=Dissolve, parameterName=Dissolve U, sliderMax=3}
   B=ScalarParameter() {defaultValue=1, group=Dissolve, parameterName=Dissolve V, sliderMax=3}
  )
  n33 = ScalarParameter() {group=Dissolve, parameterName=Dissolve Range Min, sliderMax=2, sliderMin=-2}
  n34 = ScalarParameter() {defaultValue=0.5, group=Dissolve, parameterName=Dissolve Range Max, sliderMax=2, sliderMin=-2}
  n35 = Power(
   Base=ComponentMask(None=TextureCoordinate()) {r=true}
   Exp=ScalarParameter() {defaultValue=1, group=Dissolve, parameterName=Dissolve Power, sliderMax=5}
  ) {constExponent=5}
  n36 = Power(
   Base=TextureSampleParameter2D(UVs=Multiply(A=@n31, B=@n32)) {group=Dissolve, parameterName=Dissolve Texture, texture=&T_Inky_Smoke_Tile}.R
   Exp=Saturate(None=LinearInterpolate(A=@n33, B=@n34, Alpha=@n35) {constA=0.1, constB=2})
  )
  n37 = Multiply(A=@n28, B=Saturate(None=LinearInterpolate(A=@n29, B=@n30, Alpha=@n36) {constA=-2, constB=2}))
  n38 = ComponentMask(None=TextureCoordinate()) {g=true}
  n39 = OneMinus(None=@n38)
  n40 = Min(A=OneMinus(None=Power(Base=@n38) {constExponent=10}), B=OneMinus(None=Power(Base=@n39) {constExponent=10}))
  n41 = SmoothStep(
   Min=ScalarParameter() {group=Opacity, parameterName=Edge Hardness Min, sliderMax=1}
   Max=ScalarParameter() {defaultValue=0.9, group=Opacity, parameterName=Edge Hardness Max, sliderMax=1}
   Value=DepthFade(Opacity=Multiply(A=@n37, B=@n40), FadeDistance=ScalarParameter() {defaultValue=10, group=Opacity, parameterName=FadeDistance, sliderMax=50})
  ) {constMax=0.1}
  n42 = ComponentMask(None=TextureCoordinate()) {g=true}
  n43 = DynamicParameter() {defaultValue=<rgba 1 0.5 1 1.5>, paramNames [Mask Pan, Sin Amplitude, Core Intensity, Growth]}
  n44 = Subtract(A=Add(A=@n42, B=LinearInterpolate(Alpha=@n43.Mask Pan) {constA=-1}) {constB=0}) {constB=0.5}
  n45 = OneMinus(None=@n42)
  n46 = LinearInterpolate(Alpha=@n43.Sin Amplitude) {constA=-1}
  n47 = Multiply(A=@n44, B=Subtract(A=Add(A=@n45, B=@n46) {constB=0}) {constB=0.5})
  n48 = DotProduct(A=NamedRerouteUsage(), B=CameraVectorWS())
  Multiply(
   A=Multiply(A=Multiply(A=@n7.A, B=@n41), B=SmoothStep(Value=@n47) {constMax=0.7, constMin=0.4})
   B=Saturate(None=Multiply(A=@n48) {constB=1.5})
  )
  Multiply(A=@n28, B=Round(None=@n36))
  n49 = MaterialFunctionCall(
   WidthBase=ScalarParameter() {defaultValue=250, parameterName=WidthBase}
   WidthTip=ScalarParameter() {defaultValue=100, parameterName=WidthTip}
  ) {materialFunction=&SplineThicken}
  NamedRerouteDeclaration(None=@n49.WorldPositionOffset) {name=STWPO, nodeColor=<rgba 0 0.6719 1 1>}
  NamedRerouteDeclaration(None=@n49.Normal) {name=STN, nodeColor=<rgba 0.9922 1 0 1>}
  OneMinus(None=Multiply(A=Abs(None=@n44)) {constB=2})
  StaticSwitchParameter(True=@n40, False=Constant() {r=1}) {group=Opacity, parameterName=End fade}
 ]
}
