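# Niagara system that covers a character's body in frosting and colored sprinkle meshes around an impact point.
# Both emitters burst once, partition particles across sampled bones or sockets, and reattach them each update using per-partition or cylinder-shaped offsets.
# A region mask sized from impact size zeroes the scale of particles outside the hit area.
# System timeline curves shrink both layers and fade frosting alpha, with character size adjusting frosting scale and sprinkle spread.
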NiagaraSystem /BuzzCool/Assets/FX/NS_BuzzCool_Body_Covered
dependencies {
 usedDataInterfaces [&NiagaraDataInterfaceCurve, &NiagaraDataInterfaceArrayColor, &NiagaraDataInterfaceArrayFloat]
 usedDynamicInputs [
  &Multiply_Float
  &RandomRangeFloat
  &Multiply_VectorByFloat
  &UniformRangedFloat
 ]
 usedModules [
  &self:SystemSpawnScript.SystemScriptSource.SystemScriptGraph.NiagaraNodeAssignment_1.SetVariables_#1
  &self:SystemSpawnScript.SystemScriptSource.SystemScriptGraph.NiagaraNodeAssignment_7.SetVariables_#2
  &SystemState
  &Timeline
  &self:SetPeeloutFXArrayIndex
  &self:SystemSpawnScript.SystemScriptSource.SystemScriptGraph.NiagaraNodeAssignment_3.SetVariables_#3
 ]
}
emitters {
 Mesh_Frosting {
  emitterUpdateScript [
   &EmitterState()
   &SpawnBurst_Instantaneous(Spawn Count=7, Use Spawn Probability=false, UseLoopCountLimit=true)
  ]
  particleSpawnScript [
   &PartitionParticles(Number Of Partitions=7)
   &InitializeParticle(
    Color=<rgba 1 0 0.302 1>
    Mesh Scale=&Multiply_Vector(
     A=<0.08 0.08 0.15>
     B=&SelectVectorFromArray(
      Direct Array Index=$Particles.PartitionParticles.Partition
      Vector Selection Array=DataInterface() {propertyValues {FloatData [<1.5 1.5 0.6>, <1.8 1.8 0.7>, <0.7 0.7 0.4>, <1 0.85 0.75>, <0.8 0.8 0.7>, <0.92 0.9 0.8>, <1.3 0.85 0.75>]}}
     )
    )
    Mesh Scale Mode=::NewEnumerator4
    Sprite Rotation Mode=::NewEnumerator2
    Sprite Size Mode=::NewEnumerator1
    Uniform Sprite Size Max=6
    Uniform Sprite Size Min=2
    UsePositionOffset=false
    Write Material Random=true
   )
   &SkeletalMeshLocation(
    Bone / Socket Index=$Particles.PartitionParticles.Partition
    Bone Sampling Mode=::NewEnumerator4
    MaskSpawnGroup=false
    Orientation Sampling=::NewEnumerator1
    Sampled Position Offset=&SelectVectorFromArray(
     Direct Array Index=$Particles.PartitionParticles.Partition
     Vector Selection Array=DataInterface() {propertyValues {FloatData [<-8 0 0>, <-10 -3 0>, <-4 -4.3 0>, <4 0 -1.6>, <-3 -0.6 1.4>, <0 0 1>, <0 0 0>]}}
    )
    Skeletal Mesh=$User.SkeletalMesh
    Use Sampled Mesh Attribute as Ribbon ID=false
   ) {module=SkeletalMeshLocation001}
   &RegionMask(
    Region Origin:Vector3f=&TransformPosition(Position=$User.ImpactPosition, Source Space=::World, Destination Space=::Simulation)
    Region Radius:NiagaraFloat=&Add_Float(A=25, B=&Divide_Float(A=$User.ImpactSize, B=2))
    UseOffset=false
    UseRotation=false
   )
   &InitialMeshOrientation(
    Mesh Axis To Orient=<0 0 1>
    Mesh Orientation Mode=::NewEnumerator2
    Orientation Vector=&SelectVectorFromArray(
     Direct Array Index=$Particles.PartitionParticles.Partition
     Vector Selection Array=DataInterface() {propertyValues {FloatData [<1 0 0>, <1 0 0>, <0 0 0>, <-1 0 0>, <1 0 0>, <1 0 0>, <-1 0 0>]}}
    )
    Rotation=&SelectVectorFromArray(
     Direct Array Index=$Particles.PartitionParticles.Partition
     Vector Selection Array=DataInterface() {propertyValues {FloatData [<0 0 -0.1>, <0 0 0.6>, <0 0 0>, <0 0 0.4>, <0 -0.02 0.5>, <0 0 0>, <0 0 0.8>]}}
    )
    Use Rotation Vector=true
   )
   /SetVariables(Particles.Scale=&Multiply_Vector(
     A=&Multiply_Vector(A=$Particles.Scale, B=$System.Adjust Scale by Character Size)
     B=&VectorFromFloat(Value=&MakeCustomFloatFromBool(Bool=$Particles.RegionMask.IsInsideMask, True Float=1, False Float=0))
    )) {isSetParameters=true, module=SetVariables_#4}
  ]
  particleUpdateScript [
   &ParticleState(Kill Particles When Lifetime Has Elapsed=false)
   &SkeletalMeshLocation(
    Bone / Socket Index=$Particles.PartitionParticles.Partition
    Bone Sampling Mode=::NewEnumerator4
    MaskSpawnGroup=false
    Sampled Position Offset=&SelectVectorFromArray(
     Direct Array Index=$Particles.PartitionParticles.Partition
     Vector Selection Array=DataInterface() {propertyValues {FloatData [<-8 0 0>, <-10 -3 0>, <-4 -4.3 0>, <4 0 -1.6>, <-3 -0.6 1.4>, <0 0 1>, <0 0 0>]}}
    )
    Skeletal Mesh=$User.SkeletalMesh
    Use Sampled Mesh Attribute as Ribbon ID=false
   )
   &ScaleMeshSize(Scale Factor=&Multiply_Vector(
     A=&MakeVector(
      X=&MakeFloatFromVector(
       Vector=&VectorFromFloat(Value=&FloatFromCurve(
         FloatCurve=DataInterface() {propertyValues {Curve {
            defaultValue=1
            keys [
             {interpMode=RCIM_Cubic, tangentMode=RCTM_User, time=0.5, value=1}
             {arriveTangent=-2.8297, interpMode=RCIM_Cubic, leaveTangent=-2.8297, tangentMode=RCTM_User, time=1, value=0}
            ]
           }}}
         CurveIndex=$System.Timeline.NormalizedTime
        ))
       Channel=::NewEnumerator0
      )
      Y=&MakeFloatFromVector(
       Vector=&VectorFromFloat(Value=&FloatFromCurve(
         FloatCurve=DataInterface() {propertyValues {Curve {
            defaultValue=1
            keys [
             {interpMode=RCIM_Cubic, tangentMode=RCTM_User, time=0.5, value=1}
             {arriveTangent=-2.612, interpMode=RCIM_Cubic, leaveTangent=-2.612, tangentMode=RCTM_User, time=1, value=0}
            ]
           }}}
         CurveIndex=$System.Timeline.NormalizedTime
        ))
       Channel=::NewEnumerator0
      )
      Z=1
     )
     B=<1 1 1>
    ))
   &ScaleColor(
    Scale Alpha=&FloatFromCurve(
     FloatCurve=DataInterface() {propertyValues {Curve {keys [<at 0 1 RCIM_Cubic>, {arriveTangent=-0.2157, interpMode=RCIM_Cubic, leaveTangent=-0.2157, tangentMode=RCTM_User, time=1, value=-1}]}}}
     CurveIndex=$System.Timeline.NormalizedTime
    )
    ScaleA=true
    ScaleRGB=true
   )
  ]
  rendererClasses [&NiagaraMeshRendererProperties]
  renderers [{rendererClass=&NiagaraMeshRendererProperties, rendererIndex=0}]
 }
 Sprinkles {
  emitterUpdateScript [
   &EmitterState()
   &SpawnBurst_Instantaneous(
    Spawn Count=&RandomRangeInt(Minimum=25, Maximum=46)
    Use Spawn Probability=false
    UseLoopCountLimit=true
   )
  ]
  particleSpawnScript [
   &PartitionParticles(Number Of Partitions=5)
   &InitializeParticle(
    Color=&SelectLinearColorFromArray(
     Array Sampling Mode=::NewEnumerator0
     Direct Array Index=?
     Color Selection Array=DataInterface() {propertyValues {ColorData [
        <rgba 0.7835 0.0307 0.2159 1>
        <rgba 0.3005 0.5906 0.6939 1>
        <rgba 0.4233 0.4969 0.159 1>
        <rgba 0.8714 0.8714 0.9047 1>
        <rgba 0.7231 0.2159 0.0222 1>
        <rgba 0.7605 0.6308 0.0222 1>
       ]}}
     Random Seed=0
    )
    Mesh Scale=<1.8 1.2 1.2>
    Mesh Scale Mode=::NewEnumerator4
    UsePositionOffset=false
    Write Material Random=true
   )
   &SkeletalMeshLocation(
    Bone / Socket Index=$Particles.PartitionParticles.Partition
    Bone Sampling Mode=::NewEnumerator4
    MaskSpawnGroup=false
    Orientation Sampling=::NewEnumerator1
    Sampled Position Offset=$Particles.ShapeLocation.ShapeVector
    Skeletal Mesh=$User.SkeletalMesh
    Use Sampled Mesh Attribute as Ribbon ID=false
   ) {module=SkeletalMeshLocation001}
   &InitialMeshOrientation(Use Rotation Vector=false)
   &RegionMask(
    Region Origin:Vector3f=&TransformPosition(Position=$User.ImpactPosition, Source Space=::World, Destination Space=::Simulation)
    Region Radius:NiagaraFloat=&Add_Float(A=25, B=&Divide_Float(A=$User.ImpactSize, B=2))
    UseOffset=false
    UseRotation=false
   )
   &ShapeLocation(
    Apply To Particle Position=::NewEnumerator1
    Cylinder Height=&SelectFloatFromArray(
     Direct Array Index=$Particles.PartitionParticles.Partition
     Float Selection Array=DataInterface() {propertyValues {FloatData [20,14,8,-20,20,40,-40]}}
    )
    Cylinder Height Midpoint=0
    Cylinder Radius=&SelectFloatFromArray(
     Direct Array Index=$Particles.PartitionParticles.Partition
     Float Selection Array=DataInterface() {propertyValues {FloatData [8,10,12,5,5,5,5]}}
    )
    MaskSpawnGroup=false
    Non Uniform Scale=&Multiply_Vector(A=<1 1 1>, B=$System.Adjust Scale by Character Size)
    Override Seed=false
    Rotation Coordinate Space=::World
    Rotation Mode=::NewEnumerator1
    Shape Primitive=::NewEnumerator1
    Surface Only=true
    Use Endcaps In Surface Only Mode=false
    UseOwnerScale=true
   )
   &CameraOffset(Camera Offset Amount=8)
   /SetVariables(Particles.Scale=&VectorFromFloat(Value=&MakeCustomFloatFromBool(Bool=$Particles.RegionMask.IsInsideMask, True Float=1, False Float=0))) {isSetParameters=true, module=SetVariables_#4}
  ]
  particleUpdateScript [
   &ParticleState(Kill Particles When Lifetime Has Elapsed=false)
   &SkeletalMeshLocation(
    Bone / Socket Index=$Particles.PartitionParticles.Partition
    Bone Sampling Mode=::NewEnumerator4
    MaskSpawnGroup=false
    Sampled Position Offset=$Particles.ShapeLocation.ShapeVector
    Skeletal Mesh=$User.SkeletalMesh
    Use Sampled Mesh Attribute as Ribbon ID=false
   )
   &ScaleMeshSize(Scale Factor=&VectorFromFloat(Value=&FloatFromCurve(
      FloatCurve=DataInterface() {propertyValues {Curve {
         defaultValue=1
         keys [
          {arriveTangent=-1.1643, interpMode=RCIM_Cubic, leaveTangent=-1.1643, tangentMode=RCTM_User, time=0.3, value=1}
          {interpMode=RCIM_Cubic, leaveTangent=-0.5044, time=0.8, value=0}
         ]
        }}}
      CurveIndex=$System.Timeline.NormalizedTime
     )))
  ]
  rendererClasses [&NiagaraMeshRendererProperties]
  renderers [{rendererClass=&NiagaraMeshRendererProperties, rendererIndex=0}]
 }
}
summary {
 systemName=NS_BuzzCool_Body_Covered
 emitters [
  {emitterName=Mesh_Frosting, rendererClasses [&NiagaraMeshRendererProperties]}
  {emitterName=Sprinkles, rendererClasses [&NiagaraMeshRendererProperties]}
 ]
 userVariables [
  User.CharacterSizeID {type=NiagaraInt32}
  User.Duration {type=NiagaraFloat, defaultValue=5, description=""}
  User.ImpactPosition {type=NiagaraPosition, description="", defaultValue {}}
  User.ImpactSize {type=NiagaraFloat, defaultValue=500, description=""}
  User.Mesh Scale Array {
   type=NiagaraDataInterfaceArrayFloat3
   description=""
   defaultValue {
    struct=NiagaraExt_VariableValue_DataInterface
    value {
     dataInterface=&self:NiagaraDataInterfaceArrayFloat3_0
     dataInterfaceClass=&NiagaraDataInterfaceArrayFloat3
    }
   }
  }
  User.RemoveTimer {type=NiagaraFloat, defaultValue=0, description=""}
  User.SkeletalMesh {
   type=NiagaraDataInterfaceSkeletalMesh
   defaultValue {
    struct=NiagaraExt_VariableValue_DataInterface
    value {dataInterface=&self:NiagaraDataInterfaceSkeletalMesh_0}
   }
  }
 ]
}
