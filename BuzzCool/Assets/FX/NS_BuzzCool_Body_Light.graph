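# Niagara system that supplies a magenta-purple glow light for the BuzzCool body, using a single Light emitter with only a light renderer.
# An instantaneous burst spawns the particle, and Light_Attributes sets its HDR color, radius and falloff, with the light offset slightly downward.
# Lifetime expiry does not kill the particle, so the light persists, with infinitely lived particles allowed to die when the emitter deactivates.
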NiagaraSystem /BuzzCool/Assets/FX/NS_BuzzCool_Body_Light
dependencies {usedDataInterfaces [], usedDynamicInputs [], usedModules [&SystemState]}
emitters {Light {
  emitterUpdateScript [
   &EmitterState()
   &SpawnBurst_Instantaneous(Use Spawn Probability=false, UseLoopCountLimit=true)
  ]
  particleSpawnScript [&InitializeParticle(Color Mode=::NewEnumerator0, Lifetime=5, UsePositionOffset=false, Write Material Random=true)]
  particleUpdateScript [
   &ParticleState(Kill Particles When Lifetime Has Elapsed=false, Let Infinitely Lived Particles Die When Emitter Deactivates=true)
   &Light_Attributes(
    Color=1
    Color Mode=0
    Falloff Exponent=1
    Light Color:LinearColor=<rgba 19.2878 0 30 1>
    Light Exponent=2
    Light Position Offset:Vector3f=<0 0 -20>
    Light Radius=150
    Position=1
    Radius=1
    UsePositionOffset=true
   )
  ]
  rendererClasses [&NiagaraLightRendererProperties]
  renderers [{rendererClass=&NiagaraLightRendererProperties, rendererIndex=0}]
 }}
summary {
 systemName=NS_BuzzCool_Body_Light
 emitters [{emitterName=Light, rendererClasses [&NiagaraLightRendererProperties]}]
 userVariables []
}
