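# Niagara one-shot defeat effect for BuzzCool: a purple-magenta burst of flares, embers, smoke and spline meshes.
# Burst emitters layer curl-noise embers, expanding flare meshes and sprites, and rising mesh and wispy smoke, with spawn groups varying several layers.
# A system CameraFade value scales brightness or opacity across several emitters, and one flare also follows time of day.
# Two alternate donut-smoke emitters, using Hermite-spline motion or particle-read spawning, are disabled.
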NiagaraSystem /BuzzCool/Assets/FX/NS_BuzzCool_Defeated
dependencies {
 usedDataInterfaces [
  &NiagaraDataInterfaceCurve
  &NiagaraDataInterfaceColorCurve
  &NiagaraDataInterfaceArrayFloat
  &NiagaraDataInterfaceArrayFloat3
  &NiagaraDataInterfaceCamera
  &NiagaraDataInterfaceParticleRead
  &NiagaraDataInterfaceSpriteRendererInfo
  &NiagaraDataInterfaceMeshRendererInfo
 ]
 usedDynamicInputs [
  &Vector2DFromFloat
  &SimulationPosition
  &LerpLinearColors
  &DistanceBetweenPositions
  &ReturnCameraPosition
 ]
 usedModules [
  &SystemState
  &self:SystemSpawnScript.SystemScriptSource.SystemScriptGraph.NiagaraNodeAssignment_0.SetVariables_#1
 ]
}
emitters {
 Embers {
  emitterUpdateScript [
   &EmitterState()
   &SpawnBurst_Instantaneous(
    Spawn Count=&RandomRangeInt(Minimum=22, Maximum=45)
    Use Spawn Probability=false
    UseLoopCountLimit=false
   )
  ]
  particleSpawnScript [
   /SetVariables(Particles.RandomInt:NiagaraInt32=&RandomRangeInt(Minimum=0, Maximum=1)) {isSetParameters=true}
   /SetVariables(Particles.VisibilityTag=$Particles.RandomInt) {isSetParameters=true}
   &InitializeParticle(
    Color=<rgba 2 2 2 1>
    Lifetime Max=1.4
    Lifetime Min=0.7
    Lifetime Mode=::NewEnumerator1
    Sprite Size Mode=::NewEnumerator3
    Uniform Sprite Size=&Lerp_Float(
     A=&RandomRangeFloat(Minimum=4.5, Maximum=12)
     B=&RandomRangeFloat(Minimum=40, Maximum=80)
     Alpha=&MakeFloatFromInt(Int=$Particles.RandomInt)
    )
    UsePositionOffset=false
    Write Material Random=false
   )
   &ShapeLocation(MaskSpawnGroup=false, Override Seed=false, Sphere Radius=5, UseOwnerScale=true)
   &AddVelocity(
    Module.Use Falloff Distance=false
    Override Seed=false
    Velocity Mode=::NewEnumerator1
    Velocity Speed=&RandomRangeFloat(Minimum=375, Maximum=800)
   )
   /SetVariables(Particles.RandomAtSpawn:NiagaraFloat=&RandomRangeFloat(Minimum=0, Maximum=1)) {isSetParameters=true}
   &CameraOffset(Camera Offset Amount=50)
  ]
  particleUpdateScript [
   &ParticleState()
   /Niagara/Modules/Update/Forces/CurlNoiseForce(
    Noise Frequency=8
    Noise Strength=&Multiply_Float(
     A=1760
     B=&FloatFromCurve(FloatCurve=DataInterface() {propertyValues {Curve {keys [
          {interpMode=RCIM_Cubic, leaveTangentWeight=0.3092, tangentMode=RCTM_User, time=0.1, value=1}
          {arriveTangent=-2.7872, arriveTangentWeight=0.8575, interpMode=RCIM_Cubic, leaveTangent=-2.7872, tangentMode=RCTM_User, time=1, value=0}
         ]}}})
    )
    Pan Noise=true
    Pan Noise Field=<0 0 1>
   )
   &Drag(Drag=4, Use Linear Drag=true, Use Rotational Drag=true)
   &SolveForcesAndVelocity(Clamp Velocity=false, Limit Acceleration=false)
   &ScaleColor(
    Scale Mode=1
    Scale RGBA=&Vector4FromCurve(Vector4Curve=DataInterface() {propertyValues {
       AlphaCurve {keys [<at 0.0173 0>,<at 0.1419 1>,<at 0.4612 0.854>,<at 1 0>]}
       BlueCurve {keys [<at 0.1773 1>,<at 0.534 1>,<at 1 0>]}
       GreenCurve {keys [<at 0.1773 0.3873>,<at 0.534 0>,<at 1 0>]}
       RedCurve {keys [<at 0.1773 0.0516>,<at 0.534 0.6006>,<at 1 0>]}
      }})
    ScaleRGBA=true
   )
   &ScaleColor(
    Scale Alpha=&Lerp_Float(A=1, B=0.5, Alpha=$System.CameraFade)
    Scale RGB=&MakeVectorFromLinearColorRGB(Color=/Niagara/DynamicInputs/TypeConversions/MakeLinearColorFromFloat(Value:NiagaraFloat=&Lerp_Float(A=1, B=0.7, Alpha=$System.CameraFade)))
    ScaleA=true
    ScaleRGB=true
   ) {module=ScaleColor001}
  ]
  rendererClasses [&NiagaraSpriteRendererProperties]
  renderers [
   {rendererClass=&NiagaraSpriteRendererProperties}
   {rendererClass=&NiagaraSpriteRendererProperties, rendererIndex=1}
  ]
 }
 Flare002 {
  emitterUpdateScript [
   &EmitterState()
   &SpawnBurst_Instantaneous(
    Spawn Count=&RandomRangeInt(Minimum=6, Maximum=9)
    Spawn Time=0.02
    Use Spawn Probability=false
    UseLoopCountLimit=false
   )
  ]
  particleSpawnScript [
   &InitializeParticle(
    Color=<rgba 1 1 1 2>
    Lifetime=0.35
    Sprite Size Max=<260 450>
    Sprite Size Min=<240 375>
    Sprite Size Mode=::NewEnumerator2
    UsePositionOffset=false
    Write Material Random=true
   )
   &DynamicMaterialParameters(
    Index 0 Param 1=&RandomRangeFloat(Minimum=0.4, Maximum=0.6)
    Index 0 Param 2=1
    Index 0 Param 3=1
    Index 0 Param 4=1
    Index 1 Param 1=100
    Module.Param0WriteEnabled=true
    Module.Param0WriteEnabled1=true
    Module.Param1WriteEnabled=true
    Module.Param1WriteEnabled1=true
    Module.Param2WriteEnabled=true
    Module.Param2WriteEnabled1=true
    Module.Param3WriteEnabled=true
    Module.Param3WriteEnabled1=true
    Write Parameter Index 1=true
   )
   &CameraOffset(Camera Offset Amount=50)
   &AddVelocity(Module.Use Falloff Distance=false, Override Seed=false, Velocity Mode=::NewEnumerator1, Velocity Speed=50)
  ]
  particleUpdateScript [
   &ParticleState()
   &ScaleColor(
    Scale Alpha=&Multiply_Float(
     A=1
     B=&FloatFromCurve(
      FloatCurve=DataInterface() {propertyValues {Curve {keys [{arriveTangent=-1.3786, interpMode=RCIM_Cubic, leaveTangent=-1.3786, tangentMode=RCTM_User, time=0, value=1}, <at 1 0>]}}}
      Scale Curve=&Lerp_Float(A=3, B=0.4, Alpha=$System.CameraFade)
     )
    )
    Scale RGB=&Multiply_VectorByFloat(
     Vector=&MakeVectorFromLinearColorRGB(Color=&ColorFromCurve(ColorCurve=DataInterface() {propertyValues {
         AlphaCurve {keys []}
         BlueCurve {keys [<at 0.1796 8 RCIM_Cubic>,<at 0.9157 1>]}
         GreenCurve {keys [<at 0.1796 0 RCIM_Cubic>,<at 0.9157 0>]}
         RedCurve {keys [<at 0.1796 6.9788 RCIM_Cubic>,<at 0.9157 0.5859>]}
        }}))
     Float=&FNTimeOfDayScalar(Night Scale=0.4, Morning Scale=1, Evening Scale=1)
    )
    ScaleA=true
    ScaleRGB=true
   )
   &ScaleSpriteSize(
    Uniform Curve Scale=0.6
    Uniform Curve Sprite Scale=DataInterface() {propertyValues {Curve {keys [
        {arriveTangent=2.1239, interpMode=RCIM_Cubic, leaveTangent=2.1239, tangentMode=RCTM_User, time=0, value=0}
        {arriveTangent=0.0224, interpMode=RCIM_Cubic, leaveTangent=0.0224, tangentMode=RCTM_User, time=1, value=1}
       ]}}}
   )
   &SolveForcesAndVelocity(Clamp Velocity=false, Limit Acceleration=false)
  ]
  rendererClasses [&NiagaraSpriteRendererProperties]
  renderers [{rendererClass=&NiagaraSpriteRendererProperties}]
 }
 NE_Flare {
  emitterUpdateScript [
   &EmitterState()
   &SpawnBurst_Instantaneous(Use Spawn Probability=false, UseLoopCountLimit=false)
   &SpawnBurst_Instantaneous(Spawn Group=1, Use Spawn Probability=false, UseLoopCountLimit=false) {module=SpawnBurst_Instantaneous001}
   &SpawnBurst_Instantaneous(Spawn Group=2, Use Spawn Probability=false, UseLoopCountLimit=false) {module=SpawnBurst_Instantaneous002}
   &SpawnBurst_Instantaneous(Spawn Group=3, Use Spawn Probability=false, UseLoopCountLimit=false) {module=SpawnBurst_Instantaneous003}
  ]
  particleSpawnScript [
   &InitializeParticle(
    Color=<rgba 2.7387 0 4 1>
    Lifetime=&SelectFloatFromArray(
     Direct Array Index=&SpawnGroup()
     Float Selection Array=DataInterface() {propertyValues {FloatData [0.3125,0.5,0.375,0.375]}}
    )
    Mass Mode=::NewEnumerator1
    Mesh Scale=&VectorFromFloat(Value=&Multiply_Float(A=&MaskFloatBySpawnGroup(Masked Float=10, Passthrough Float=3.8), B=0.5))
    Mesh Scale Mode=::NewEnumerator4
    Sprite Size Mode=::NewEnumerator4
    UsePositionOffset=true
    Write Material Random=true
   )
   /SetVariables(Particles.VisibilityTag=&SpawnGroup()) {isSetParameters=true}
   &InitialMeshOrientation(Orientation Axis=<0 0 1>, Orientation Vector=<0 0 1>, Rotation=<0 0.25 0>, Use Orientation Vector=true, Use Rotation Vector=true)
   &InitialMeshOrientation(
    Orientation Axis=<0 0 0>
    Orientation Vector=<0 0 0>
    Rotation=&RandomRangeVector(Minimum=<0 0 0>, Maximum=<1 0 0>)
    Use Orientation Vector=false
    Use Rotation Vector=true
   ) {module=InitialMeshOrientation001}
   /SetVariables(Particles.BigFlareOrient:Quat4f=&OrientMeshToVector(Look At Direction=<0 0 0>, Reference Axis Vector=<1 0 0>)) {isSetParameters=true}
   /SetVariables(Particles.BigFlareDynmat:Vector4f=<1 0 0 0>) {isSetParameters=true}
  ]
  particleUpdateScript [
   &ParticleState()
   &ScaleMeshSize(Scale Factor=&VectorFromFloat(Value=&MaskFloatBySpawnGroup(
      Masked Float=&FloatFromCurve(FloatCurve=DataInterface() {propertyValues {Curve {keys [{arriveTangent=5.2157, interpMode=RCIM_Cubic, leaveTangent=5.2157, tangentMode=RCTM_User, time=0, value=0}, <at 0.4 1 RCIM_Cubic>]}}})
      Passthrough Float=&FloatFromCurve(FloatCurve=DataInterface() {propertyValues {Curve {keys [
           {interpMode=RCIM_Cubic, tangentMode=RCTM_Break, time=0, value=0.2}
           {arriveTangent=2.9365, interpMode=RCIM_Cubic, leaveTangent=2.9365, tangentMode=RCTM_User, time=0.1, value=1}
           {interpMode=RCIM_Cubic, tangentMode=RCTM_Break, time=1, value=2}
          ]}}})
     )))
   &ScaleColor(
    Scale Alpha=&MaskFloatBySpawnGroup(
     Masked Float=0
     Passthrough Float=&FloatFromCurve(FloatCurve=DataInterface() {propertyValues {Curve {keys [
          {interpMode=RCIM_Cubic, tangentMode=RCTM_User, time=0, value=1}
          {arriveTangent=-2.6886, interpMode=RCIM_Cubic, leaveTangent=-2.6886, tangentMode=RCTM_User, time=1, value=0}
         ]}}})
    )
    Scale RGB=&MaskVectorBySpawnGroup(
     Masked Vector=&VectorFromFloat(Value=&FloatFromCurve(FloatCurve=DataInterface() {propertyValues {Curve {keys [<at 0 0.7>,<at 0.7729 0.7>,<at 1 0>]}}}, Scale Curve=0.2))
     Passthrough Vector=&Multiply_Vector(
      A=&MakeVectorFromLinearColorRGB(Color=&ColorFromCurve(ColorCurve=DataInterface() {propertyValues {
          AlphaCurve {keys [<at 0 1>,<at 0.4696 1>,<at 1 0>]}
          BlueCurve {keys [<at 0 1>,<at 0.2681 1>,<at 0.4696 0.6979>,<at 1 0.3698>]}
          GreenCurve {keys [<at 0 0.05>,<at 0.2681 0>,<at 0.4696 0.0698>,<at 1 0.1701>]}
          RedCurve {keys [<at 0 0.6117>,<at 0.2681 0.3826>,<at 0.4696 0.2118>,<at 1 0.037>]}
         }}))
      B=&VectorFromFloat(Value=&FloatFromCurve(
        FloatCurve=DataInterface() {propertyValues {Curve {keys [
            {interpMode=RCIM_Cubic, tangentMode=RCTM_Break, time=0, value=10}
            {arriveTangent=-6.6072, interpMode=RCIM_Cubic, leaveTangent=-6.6072, tangentMode=RCTM_User, time=0.15, value=3}
            {interpMode=RCIM_Cubic, tangentMode=RCTM_Break, time=1, value=1}
           ]}}}
        Scale Curve=&Lerp_Float(A=0.1, B=1, Alpha=$System.CameraFadeAlpha)
       ))
     )
    )
    ScaleA=true
    ScaleRGB=true
   )
   &DynamicMaterialParameters(
    Index 0 Param 1=&FloatFromCurve(FloatCurve=DataInterface() {propertyValues {Curve {keys [{arriveTangent=-2.1015, interpMode=RCIM_Cubic, leaveTangent=-2.1015, tangentMode=RCTM_User, time=0, value=0.5}, <at 1 -0.15>]}}})
    Index 0 Param 2=&FloatFromCurve(FloatCurve=DataInterface() {propertyValues {Curve {keys [{arriveTangent=-0.3464, interpMode=RCIM_Cubic, leaveTangent=-0.3464, tangentMode=RCTM_User, time=0, value=1}, <at 1 0.9>]}}})
    Index 0 Param 3=&FloatFromCurve(FloatCurve=DataInterface() {propertyValues {Curve {keys [
         {interpMode=RCIM_Cubic, tangentMode=RCTM_Break, time=0, value=0}
         {arriveTangent=0.4478, interpMode=RCIM_Cubic, leaveTangent=0.4478, tangentMode=RCTM_User, time=0.15, value=1}
         {interpMode=RCIM_Cubic, tangentMode=RCTM_Break, time=1, value=1.2}
        ]}}})
    Index 3 Param 1=1
    Module.Param0WriteEnabled=true
    Module.Param0WriteEnabled3=true
    Module.Param1WriteEnabled=true
    Module.Param1WriteEnabled3=true
    Module.Param2WriteEnabled=true
    Module.Param2WriteEnabled3=true
    Module.Param3WriteEnabled=false
    Module.Param3WriteEnabled3=true
    Write Parameter Index 3=true
   )
   /SetVariables(Particles.BigFlareScale:Vector3f=&Multiply_VectorByFloat(Vector=$Particles.Scale, Float=0.1875)) {isSetParameters=true}
   /SetVariables(Particles.BigFlareColor:LinearColor=&MultiplyLinearColor(
     Linear Color=<rgba 22.7949 0 50 1>
     Scale Factor=&MakeFloat4FromVectorAndFloat(
      "Vector (RGB)"=&VectorFromFloat(Value=0.05)
      "Float (Alpha)"=&FloatFromCurve(FloatCurve=DataInterface() {propertyValues {Curve {keys [<at 0.4 1 RCIM_Cubic>,<at 1 0>]}}})
     )
    )) {isSetParameters=true}
   /SetVariables(Particles.RefracSphereDynMat:Vector4f=&MakeVector4(
     X=&FloatFromCurve(FloatCurve=DataInterface() {propertyValues {Curve {keys [<at 0 1>,<at 0.3172 0.85>,<at 1 1>]}}})
     Y=&FloatFromCurve(FloatCurve=DataInterface() {propertyValues {Curve {keys [<at 0 15>,<at 1 0>]}}})
     Z=0
     W=0
    )) {isSetParameters=true}
   &FN_Sprite_Attributes(
    Color=&Multiply_LinearColorByFloat(Linear Color=$Particles.Color, Float=5)
    Create Color Attribute=true
    Create SpriteSize Attribute=true
    SpriteSize=&Multiply_Vector2D_ByFloat(
     Float=&Multiply_Float(
      A=&Multiply_Float(
       A=750
       B=&FloatFromCurve(FloatCurve=DataInterface() {propertyValues {Curve {keys [
            {arriveTangent=2.6957, interpMode=RCIM_Cubic, leaveTangent=2.6957, tangentMode=RCTM_User, time=0, value=0}
            {interpMode=RCIM_Cubic, tangentMode=RCTM_User, time=1, value=1}
           ]}}})
      )
      B=1
     )
     Vector2D=<1 1>
    )
    Write to Module Namespace=true
   )
  ]
  rendererClasses [&NiagaraMeshRendererProperties, &NiagaraSpriteRendererProperties]
  renderers [
   {rendererClass=&NiagaraMeshRendererProperties, rendererIndex=0}
   {rendererClass=&NiagaraMeshRendererProperties, rendererIndex=1}
   {rendererClass=&NiagaraMeshRendererProperties, rendererIndex=2}
   {rendererClass=&NiagaraSpriteRendererProperties, rendererIndex=3}
  ]
 }
 Smoke_Masked_Donut {
  bEnabled=false
  emitterUpdateScript [
   &EmitterState()
   &SpawnBurst_Instantaneous(Spawn Count=12, Use Spawn Probability=false, UseLoopCountLimit=false)
  ]
  particleSpawnScript [
   &InitializeParticle(
    Color Maximum=<rgba 0.6094 0.1075 0.3926 1>
    Color Minimum=<rgba 0.6094 0.1075 0.3926 0.7>
    Color Mode=::NewEnumerator2
    Lifetime=3
    Mesh Scale Mode=::NewEnumerator1
    Mesh Uniform Scale Max=2.75
    Mesh Uniform Scale Min=2.5
    UsePositionOffset=false
    Write Material Random=true
   )
   /SetVariables(Particles.SpawnFloat:NiagaraFloat=&ReturnNormalizedExecIndex(Normalized Index Includes One=false)) {isSetParameters=true}
   &InitialMeshOrientation(Use Rotation Vector=false)
   &CameraOffset(Camera Offset Amount=75)
  ]
  particleUpdateScript [
   &ParticleState()
   /SetVariables(Particles.StartLocation:NiagaraFloat=&ClampFloat(Float=&Subtract_Float(
      A=&FloatFromCurve(
       FloatCurve=DataInterface() {propertyValues {Curve {keys [
           {arriveTangent=-0.8959, interpMode=RCIM_Cubic, time=0.5, value=1}
           {arriveTangent=-1.3996, interpMode=RCIM_Cubic, leaveTangent=-1.3996, tangentMode=RCTM_User, time=2, value=0}
          ]}}}
       CurveIndex=$Particles.Age
      )
      B=$Particles.SpawnFloat
     ))) {isSetParameters=true}
   &Timeline(Play=&SetBoolByFloatComparison(A=$Particles.StartLocation, B=0, Comparison Type=::NewEnumerator4))
   &ShapeLocation(
    Handle Radius=0.3
    Large Radius=&FloatFromCurve(
     FloatCurve=DataInterface() {propertyValues {Curve {
        defaultValue=1
        keys [
         {arriveTangent=6.8557, interpMode=RCIM_Cubic, leaveTangent=6.8557, tangentMode=RCTM_User, time=0, value=0}
         {tangentMode=RCTM_User, time=0.4, value=1}
        ]
       }}}
     CurveIndex=$Particles.Age
     Scale Curve=65
    )
    MaskSpawnGroup=false
    Non Uniform Scale=<1 -1 1>
    Rotation Angle:NiagaraFloat=&SelectFloatFromArray(Direct Array Index=$System.RandomInt, Float Selection Array=DataInterface() {propertyValues {FloatData [0.5,0.25,0.3,0.1]}})
    Rotation Mode=::NewEnumerator1
    Shape Primitive=::NewEnumerator3
    Torus Distribution Mode:ENiagaraTorusDistributionMode=::NewEnumerator1
    U Position=$Particles.StartLocation
    UseOwnerScale=true
   )
   /Game/Effects/Niagara/Modules/Curves/FN_HermiteSpline(
    Direction Start:Vector3f=&SelectVectorFromArray(
     Direct Array Index=$System.RandomInt
     Vector Selection Array=DataInterface() {propertyValues {FloatData [<0 -750 0>, <0 100 0>, <0 -90 0>, <0 750 0>]}}
    )
    Position Blend:NiagaraFloat=$Particles.Timeline.NormalizedTime
    Position End:Vector3f=&SelectVectorFromArray(
     Direct Array Index=$System.RandomInt
     Vector Selection Array=DataInterface() {propertyValues {FloatData [<0 -150 350>, <0 150 350>, <0 -150 300>, <0 200 300>]}}
    )
    Position Start:Vector3f=&SelectVectorFromArray(
     Direct Array Index=$System.RandomInt
     Vector Selection Array=DataInterface() {propertyValues {FloatData [<0 0 65>, <0 64.5 0>, <0 62 20>, <0 64.5 0>]}}
    )
   )
   /SetVariables(Particles.Position=&Lerp_Vector(
     A=Linked() {linkedVariable {name=Particles.ShapeLocation.ShapePosition, type=NiagaraPosition}}
     B=$Output.FN_HermiteSpline.Position
     Alpha=&FloatFromCurve(
      FloatCurve=DataInterface() {propertyValues {Curve {
         defaultValue=0
         keys [{arriveTangent=4.9242, interpMode=RCIM_Cubic, leaveTangent=4.9242, tangentMode=RCTM_User, time=0, value=0}, <at 0.5 1 RCIM_Cubic>]
        }}}
      CurveIndex=$Particles.Timeline.NormalizedTime
     )
    )) {isSetParameters=true}
   &self:ScratchModule() {module=ScratchModule}
   &UpdateMeshOrientation(Rotation Rate=0.5, Rotation Vector=&RandomVector(Vector Scale=1))
   &ScaleMeshSize(Scale Factor=&VectorFromFloat(Value=&Multiply_Float(
      A=&FloatFromCurve(
       FloatCurve=DataInterface() {propertyValues {Curve {
          defaultValue=0.9
          keys [
           {arriveTangent=8.0632, interpMode=RCIM_Cubic, leaveTangent=8.0632, tangentMode=RCTM_User, time=0, value=0.8}
           {interpMode=RCIM_Cubic, leaveTangent=-2.5249, time=0.1408, value=1.2}
          ]
         }}}
       Scale Curve=&FloatFromCurve(
        FloatCurve=DataInterface() {propertyValues {Curve {keys [<at 0 1 RCIM_Cubic>, {arriveTangent=-2.4406, tangentMode=RCTM_User, time=1, value=0}]}}}
        CurveIndex=$Particles.Timeline.NormalizedTime
       )
      )
      B=&FloatFromCurve(
       FloatCurve=DataInterface() {propertyValues {Curve {keys [<at 0.5 1 RCIM_Cubic>, {arriveTangent=-4.5326, tangentMode=RCTM_User, time=1, value=0}]}}}
       CurveIndex=&ClampFloat(Float=&Subtract_Float(A=$Particles.Age, B=$Particles.SpawnFloat))
      )
     )))
   &FN_Mesh_Attributes(
    Color=&MultiplyLinearColor(
     Linear Color=$Particles.Color
     Scale Factor=&MakeVector4(
      X=1
      Y=1
      Z=1
      W=&FloatFromCurve(
       FloatCurve=DataInterface() {propertyValues {Curve {
          defaultValue=0
          keys [
           {arriveTangent=-0.3733, interpMode=RCIM_Cubic, time=0.5, value=1}
           {arriveTangent=-4.6626, interpMode=RCIM_Cubic, leaveTangent=-4.6626, tangentMode=RCTM_User, time=1, value=0}
          ]
         }}}
       CurveIndex=$Particles.Timeline.NormalizedTime
       Scale Curve=0.3
      )
     )
    )
    Create Color Attribute=true
    Create Scale Attribute=true
    Scale=&Multiply_VectorByFloat(Vector=$Particles.Scale, Float=1.3)
    Write to Module Namespace=true
   )
   &ScaleColor(
    Scale Alpha=&FloatFromCurve(
     FloatCurve=DataInterface() {propertyValues {Curve {
        defaultValue=1
        keys [
         {arriveTangent=-2.1118, interpMode=RCIM_Cubic, time=0.2, value=1}
         {arriveTangent=-1.6108, interpMode=RCIM_Cubic, leaveTangent=-1.6108, tangentMode=RCTM_User, time=0.9, value=0}
        ]
       }}}
     CurveIndex=$Particles.Timeline.NormalizedTime
    )
    ScaleA=true
    ScaleRGB=true
   )
  ]
  rendererClasses [&NiagaraMeshRendererProperties]
  renderers [
   {rendererClass=&NiagaraMeshRendererProperties, rendererIndex=0}
   {rendererClass=&NiagaraMeshRendererProperties, rendererIndex=1}
  ]
 }
 Smoke_Masked_Donut001 {
  emitterUpdateScript [
   &EmitterState()
   &SpawnBurst_Instantaneous(Spawn Count=8, Use Spawn Probability=false, UseLoopCountLimit=false)
   &SpawnBurst_Instantaneous(Spawn Count=9, Spawn Group=1, Use Spawn Probability=false, UseLoopCountLimit=false) {module=SpawnBurst_Instantaneous001}
   &SpawnBurst_Instantaneous(Spawn Count=9, Spawn Time=0.0667, Use Spawn Probability=false, UseLoopCountLimit=false) {module=SpawnBurst_Instantaneous002}
  ]
  particleSpawnScript [
   &InitializeParticle(
    Color Maximum=<rgba 0.4962 0 0.6042 1>
    Color Minimum=<rgba 0.4914 0.0006 0.625 0.7>
    Color Mode=::NewEnumerator2
    Lifetime=&MaskFloat(
     Mask Source=0
     Mask Parameter=?
     Mask=0
     Masked Float=&RandomRangeFloat(Minimum=1.3, Maximum=1.4)
     Passthrough Float=&RandomRangeFloat(Minimum=1.45, Maximum=1.7)
    )
    Mesh Scale Mode=::NewEnumerator3
    Mesh Uniform Scale=&MaskFloat(
     Mask Source=0
     Mask Parameter=?
     Mask=0
     Masked Float=&RandomRangeFloat(Minimum=3.15, Maximum=3.3)
     Passthrough Float=&RandomRangeFloat(Minimum=1.5, Maximum=3)
    )
    Sprite Size Mode=::NewEnumerator3
    Uniform Sprite Size=220
    UsePositionOffset=true
    Write Material Random=true
   )
   &ShapeLocation(
    Cylinder Height=120
    Cylinder Height Midpoint=0
    Cylinder Radius=10
    MaskSpawnGroup=false
    Offset=<0 0 -100>
    Offset Coordinate Space=::Local
    Offset Mode=::NewEnumerator0
    Override Seed=false
    Shape Primitive=::NewEnumerator1
    Surface Only=true
    UseOwnerScale=true
   )
   &InitialMeshOrientation(Use Rotation Vector=false)
   &CameraOffset(Camera Offset Amount=50)
   &AddVelocity(
    Velocity=$Particles.ShapeLocation.ShapeNormal
    Velocity Speed Scale=&MaskFloat(
     Mask Source=0
     Mask Parameter=?
     Mask=0
     Masked Float=&RandomRangeFloat(Minimum=90, Maximum=144)
     Passthrough Float=&RandomRangeFloat(Minimum=140, Maximum=180)
    )
   )
   &AddVelocity(Velocity=&RandomRangeVector(Minimum=<0 0 0>, Maximum=<0 0 400>)) {module=AddVelocity001}
   /SetVariables(Particles.GlowCameraOffset=20, Particles.SoftCameraOffset=45) {isSetParameters=true, module=SetVariables_#2}
  ]
  particleUpdateScript [
   &ParticleState()
   &Drag(Use Linear Drag=true, Use Rotational Drag=true)
   &AccelerationForce(Acceleration=<0 0 50>)
   &SolveForcesAndVelocity(Clamp Velocity=false, Limit Acceleration=false)
   &UpdateMeshOrientation(Rotation Rate=0.7, Rotation Vector=&RandomVector(Vector Scale=1))
   &ScaleMeshSize(Scale Factor=&VectorFromFloat(Value=&FloatFromCurve(FloatCurve=DataInterface() {propertyValues {Curve {
         defaultValue=1
         keys [
          {arriveTangent=9.0725, interpMode=RCIM_Cubic, leaveTangent=9.0725, tangentMode=RCTM_User, time=0, value=0.4}
          {arriveTangent=-0.4225, interpMode=RCIM_Cubic, leaveTangent=-0.4225, tangentMode=RCTM_User, time=0.1408, value=1.3}
          {arriveTangent=-0.0405, interpMode=RCIM_Cubic, leaveTangent=-0.0405, tangentMode=RCTM_User, time=1, value=0.8}
         ]
        }}})))
   &FN_Mesh_Attributes(
    Color=&MultiplyLinearColor(
     Linear Color=<rgba 0.5471 0.0755 0.5573 0.5>
     Scale Factor=&MakeVector4(
      X=1
      Y=1
      Z=1
      W=&FloatFromCurve(
       FloatCurve=DataInterface() {propertyValues {Curve {
          defaultValue=0
          keys [
           {arriveTangent=0.1155, interpMode=RCIM_Cubic, leaveTangent=0.1155, tangentMode=RCTM_User, time=0.4, value=1}
           {arriveTangent=-2.9743, interpMode=RCIM_Cubic, leaveTangent=-2.9743, tangentMode=RCTM_User, time=1, value=0}
          ]
         }}}
       Scale Curve=&Lerp_Float(A=1, B=0.4, Alpha=$System.CameraFade)
      )
     )
    )
    Create Color Attribute=true
    Create Scale Attribute=true
    Scale=&Multiply_VectorByFloat(Vector=$Particles.Scale, Float=1.2)
    Write to Module Namespace=true
   )
   &ScaleColor(
    Scale Alpha=&FloatFromCurve(
     FloatCurve=DataInterface() {propertyValues {Curve {keys [
         {arriveTangent=-0.9949, interpMode=RCIM_Cubic, leaveTangent=-0.9949, tangentMode=RCTM_User, time=0, value=1}
         {arriveTangent=-0.1641, interpMode=RCIM_Cubic, leaveTangent=-0.1641, tangentMode=RCTM_User, time=0.8, value=0.4}
        ]}}}
     Scale Curve=&Lerp_Float(A=1, B=0.7, Alpha=$System.CameraFade)
    )
    ScaleA=true
    ScaleRGB=true
   )
   /SetVariables(Particles.Glow Color=&MultiplyLinearColor(
     Linear Color=<rgba 0.2074 0.0368 0.5271 0.4>
     Scale Factor=&MakeVector4FromVectorAndFloat(
      "Vector (RGB)"=<1 1 1>
      "Float (Alpha)"=&FloatFromCurve(
       FloatCurve=DataInterface() {propertyValues {Curve {
          defaultValue=1
          keys [{interpMode=RCIM_Cubic, tangentMode=RCTM_User, time=0, value=1}, {interpMode=RCIM_Cubic, tangentMode=RCTM_User, time=1, value=0}]
         }}}
       Scale Curve=&Lerp_Float(A=1, B=0.3, Alpha=$System.CameraFade)
      )
     )
    )) {isSetParameters=true, module=SetVariables_#3}
   &DynamicMaterialParameters(
    Index 0 Param 2=50
    Module.Index0Float0Write=true
    Module.Index0Float1Write=true
    Module.Index0Float2Write=true
    Module.Index0Float3Write=true
   )
  ]
  rendererClasses [&NiagaraMeshRendererProperties, &NiagaraSpriteRendererProperties]
  renderers [
   {rendererClass=&NiagaraMeshRendererProperties, rendererIndex=0}
   {rendererClass=&NiagaraMeshRendererProperties, rendererIndex=1}
   {rendererClass=&NiagaraSpriteRendererProperties, rendererIndex=2}
  ]
 }
 Smoke_Masked_Donut002 {
  bEnabled=false
  emitterUpdateScript [
   &EmitterState()
   &SpawnParticlesFromOtherEmitter(
    Attribute Reader=DataInterface() {propertyValues {EmitterBinding {bindingMode=Other, emitterName=Hermit}}}
    Spawn Rate=40
    Use Maximum Age=false
    Use Minimum Age=false
   )
  ]
  particleSpawnScript [
   &InitializeParticle(
    Color Maximum=<rgba 0.6094 0.1075 0.3926 1>
    Color Minimum=<rgba 0.6094 0.1075 0.3926 0.7>
    Color Mode=::NewEnumerator2
    Lifetime=&FloatFromCurve(
     FloatCurve=DataInterface() {propertyValues {Curve {
        defaultValue=1
        keys [{interpMode=RCIM_Cubic, tangentMode=RCTM_User, time=0, value=1}, {interpMode=RCIM_Cubic, tangentMode=RCTM_User, time=1, value=0.1}]
       }}}
     CurveIndex=$Emitter.NormalizedLoopAge
     Scale Curve=0.4
    )
    Mesh Scale Mode=::NewEnumerator1
    Mesh Uniform Scale Max=3
    Mesh Uniform Scale Min=1.5
    Sprite Size Mode=::NewEnumerator3
    Uniform Sprite Size=275
    UsePositionOffset=false
    Write Material Random=true
   )
   &SampleParticlesFromOtherEmitter()
   &InitialMeshOrientation(Use Rotation Vector=false)
   &CameraOffset(Camera Offset Amount=20)
   /SetVariables(Particles.GlowCameraOffset=-10, Particles.SoftCameraOffset=10) {isSetParameters=true, module=SetVariables_#2}
  ]
  particleUpdateScript [
   &ParticleState()
   &SolveForcesAndVelocity(Clamp Velocity=false, Limit Acceleration=false)
   &UpdateMeshOrientation(Rotation Rate=0.5, Rotation Vector=&RandomVector(Vector Scale=1))
   &ScaleMeshSize(Scale Factor=&VectorFromFloat(Value=&FloatFromCurve(FloatCurve=DataInterface() {propertyValues {Curve {
         defaultValue=0
         keys [
          {interpMode=RCIM_Cubic, tangentMode=RCTM_User, time=0, value=1.5}
          {arriveTangent=-3.5766, interpMode=RCIM_Cubic, leaveTangent=-3.5766, tangentMode=RCTM_User, time=1, value=0.3}
         ]
        }}})))
   &ScaleSpriteSize(
    Scale Sprite Size Mode=::NewEnumerator0
    Uniform Scale Factor=&Multiply_Float(
     A=1
     B=&RemapRange(
      Input Value To Remap=&MakeFloatFromVector(Vector=$Particles.Scale, Channel=::NewEnumerator0)
      Input Min=1.5
      Input Max=3
      Output Min=0
      Output Max=1
     )
    )
   )
   &ScaleColor(
    Scale Alpha=&FloatFromCurve(
     FloatCurve=DataInterface() {propertyValues {Curve {
        defaultValue=1
        keys [
         {arriveTangent=0.1214, interpMode=RCIM_Cubic, leaveTangent=0.1214, tangentMode=RCTM_User, time=0.1, value=1}
         {arriveTangent=-0.6386, interpMode=RCIM_Cubic, leaveTangent=-0.6386, tangentMode=RCTM_User, time=1, value=0.325}
        ]
       }}}
     Scale Curve=&FloatFromCurve(
      FloatCurve=DataInterface() {propertyValues {Curve {
         defaultValue=1
         keys [
          {interpMode=RCIM_Cubic, tangentMode=RCTM_User, time=1.7, value=1}
          {arriveTangent=-0.2488, interpMode=RCIM_Cubic, leaveTangent=-0.2488, tangentMode=RCTM_User, time=2.7, value=0}
         ]
        }}}
      CurveIndex=$Emitter.Age
     )
    )
    ScaleA=true
    ScaleRGB=true
   )
   &FN_Mesh_Attributes(
    Color=&MultiplyLinearColor(
     Linear Color=$Particles.Initial.Color
     Scale Factor=&MakeVector4(
      X=1
      Y=1
      Z=1
      W=&FloatFromCurve(
       FloatCurve=DataInterface() {propertyValues {Curve {
          defaultValue=1
          keys [
           {interpMode=RCIM_Cubic, tangentMode=RCTM_User, time=0.5, value=1}
           {arriveTangent=-0.3871, interpMode=RCIM_Cubic, leaveTangent=-0.3871, tangentMode=RCTM_User, time=1, value=0}
          ]
         }}}
       Scale Curve=&FloatFromCurve(
        FloatCurve=DataInterface() {propertyValues {Curve {
           defaultValue=1
           keys [
            {arriveTangent=-1.5674, interpMode=RCIM_Cubic, leaveTangent=-1.5674, tangentMode=RCTM_User, time=1.7, value=1}
            {arriveTangent=-0.1662, interpMode=RCIM_Cubic, leaveTangent=-0.1662, tangentMode=RCTM_User, time=2.7, value=0}
           ]
          }}}
        CurveIndex=$Emitter.Age
        Scale Curve=0.4
       )
      )
     )
    )
    Create Color Attribute=true
    Create Scale Attribute=true
    Scale=&Multiply_VectorByFloat(Vector=$Particles.Scale, Float=1.2)
    Write to Module Namespace=true
   )
   /SetVariables(Particles.Glow Color=&MultiplyLinearColor(
     Linear Color=<rgba 0 1 0.1412 0.2>
     Scale Factor=&MakeVector4FromVectorAndFloat(
      "Vector (RGB)"=<1 1 1>
      "Float (Alpha)"=&FloatFromCurve(
       FloatCurve=DataInterface() {propertyValues {Curve {
          defaultValue=1
          keys [
           {arriveTangent=-2.6465, interpMode=RCIM_Cubic, leaveTangent=-2.6465, tangentMode=RCTM_User, time=0.2, value=1}
           {arriveTangent=-0.2638, interpMode=RCIM_Cubic, leaveTangent=-0.2638, tangentMode=RCTM_User, time=1, value=0}
          ]
         }}}
       Scale Curve=&FloatFromCurve(
        FloatCurve=DataInterface() {propertyValues {Curve {
           defaultValue=1
           keys [
            {interpMode=RCIM_Cubic, tangentMode=RCTM_User, time=2, value=1}
            {arriveTangent=-0.1662, interpMode=RCIM_Cubic, leaveTangent=-0.1662, tangentMode=RCTM_User, time=2.7, value=0}
           ]
          }}}
        CurveIndex=$Emitter.Age
        Scale Curve=0.7
       )
      )
     )
    )) {isSetParameters=true, module=SetVariables_#3}
  ]
  rendererClasses [&NiagaraMeshRendererProperties, &NiagaraSpriteRendererProperties]
  renderers [
   {rendererClass=&NiagaraMeshRendererProperties, rendererIndex=0}
   {rendererClass=&NiagaraMeshRendererProperties, rendererIndex=1}
   {rendererClass=&NiagaraSpriteRendererProperties, rendererIndex=2}
  ]
 }
 Spline_Meshes {
  emitterUpdateScript [
   &EmitterState()
   &SpawnBurst_Instantaneous(Use Spawn Probability=false, UseLoopCountLimit=false)
   &SpawnBurst_Instantaneous(Spawn Group=1, Use Spawn Probability=false, UseLoopCountLimit=false) {module=SpawnBurst_Instantaneous001}
  ]
  particleSpawnScript [
   /SetVariables(Particles.VisibilityTag=&SpawnGroup()) {isSetParameters=true}
   &InitializeParticle(
    Color=&MaskLinearColorBySpawnGroup(Masked Linear Color=<rgba 1 1 1 1>, Passthrough Linear Color=<rgba 0.7231 0.0482 0.7011 8>)
    Lifetime=&MaskFloatBySpawnGroup(Masked Float=2.5, Passthrough Float=2.1)
    Mesh Scale=&MaskVectorBySpawnGroup(Masked Vector=<9 1 1>, Passthrough Vector=<5.2 1 1>)
    Mesh Scale Mode=::NewEnumerator4
    Position Offset=<0 0 -100>
    UsePositionOffset=true
    Write Material Random=true
   )
   &InitialMeshOrientation(
    Mesh Axis To Orient=<0 0 -1>
    Mesh Orientation Mode=::NewEnumerator2
    Rotation=&RandomRangeVector(Minimum=<-1 0 0>, Maximum=<1 0 0>)
    Use Rotation Vector=true
   )
  ]
  particleUpdateScript [
   &ParticleState()
   &ScaleColor(
    Scale Alpha=&MaskFloatBySpawnGroup(
     Masked Float=&FloatFromCurve(FloatCurve=DataInterface() {propertyValues {Curve {keys [<at 0.3 1 RCIM_Cubic>, {arriveTangent=-3.2514, tangentMode=RCTM_User, time=1, value=0}]}}})
     Passthrough Float=&FloatFromCurve(FloatCurve=DataInterface() {propertyValues {Curve {keys [<at 0.3 1 RCIM_Cubic>, {arriveTangent=0.4587, tangentMode=RCTM_User, time=1, value=0}]}}})
    )
    Scale RGB=&MaskVectorBySpawnGroup(
     Masked Vector=&MakeVectorFromLinearColorRGB(Color=&ColorFromCurve(ColorCurve=DataInterface() {propertyValues {
         AlphaCurve {keys [<at 0 1 RCIM_Cubic>]}
         BlueCurve {keys [
           <at 0.3663 1 RCIM_Cubic>
           <at 0.5133 0.7182>
           {arriveTangent=-2.3554, interpMode=RCIM_Cubic, leaveTangent=-2.3554, time=0.6373, value=0.474}
           <at 0.7652 0.125>
           <at 0.9454 0.0938>
          ]}
         GreenCurve {keys [<at 0.3663 1 RCIM_Cubic>,<at 0.5133 0>,<at 0.6373 0 RCIM_Cubic>,<at 0.7652 0>,<at 0.9454 0>]}
         RedCurve {keys [
           <at 0.3663 1 RCIM_Cubic>
           <at 0.5133 0.5682>
           {arriveTangent=-1.9531, interpMode=RCIM_Cubic, leaveTangent=-1.9531, time=0.6373, value=0.2893}
           <at 0.7652 0.0763>
           <at 0.9454 0.0572>
          ]}
        }}))
     Passthrough Vector=<1 1 1>
    )
    ScaleA=true
    ScaleRGB=true
   )
   &DynamicMaterialParameters(
    Index 0 Param 1=&MaskFloatBySpawnGroup(
     Masked Float=&FloatFromCurve(FloatCurve=DataInterface() {propertyValues {Curve {keys [
          {arriveTangent=0.1614, interpMode=RCIM_Cubic, leaveTangent=0.1614, tangentMode=RCTM_User, time=0, value=0.2}
          {interpMode=RCIM_Cubic, leaveTangent=1.7375, time=1, value=1}
         ]}}})
     Passthrough Float=&FloatFromCurve(FloatCurve=DataInterface() {propertyValues {Curve {
         defaultValue=1
         keys [
          {arriveTangent=0.789, interpMode=RCIM_Cubic, leaveTangent=0.789, tangentMode=RCTM_User, time=0, value=0.6}
          {tangentMode=RCTM_User, time=0.6, value=1}
         ]
        }}})
    )
    Index 0 Param 2=&MaskFloatBySpawnGroup(
     Masked Float=1
     Passthrough Float=&FloatFromCurve(FloatCurve=DataInterface() {propertyValues {Curve {keys [<at 0.5 1 RCIM_Cubic>, {arriveTangent=-0.5353, tangentMode=RCTM_User, time=1, value=0.5}]}}})
    )
    Index 0 Param 3=&FloatFromCurve(FloatCurve=DataInterface() {propertyValues {Curve {
        defaultValue=1
        keys [
         {interpMode=RCIM_Cubic, tangentMode=RCTM_User, time=0, value=1}
         {arriveTangent=-1.0051, interpMode=RCIM_Cubic, leaveTangent=-1.0051, tangentMode=RCTM_User, time=1, value=0}
        ]
       }}})
    Index 0 Param 4=&RandomRangeFloat(Minimum=-0.4, Maximum=0.4)
    Index 1 Param 1=&MakeFloatFromVector(
     Vector=&AddVectorToPosition(
      Position=$Engine.Owner.Position
      Vector=&VectorFromFloat(Value=&MaskFloatBySpawnGroup(
        Masked Float=&RandomRangeFloat(Minimum=-337.5, Maximum=337.5)
        Passthrough Float=&RandomRangeFloat(Minimum=-150, Maximum=150)
       ))
     )
     Channel=::NewEnumerator0
    )
    Index 1 Param 2=&MakeFloatFromVector(
     Vector=&AddVectorToPosition(
      Position=$Engine.Owner.Position
      Vector=&VectorFromFloat(Value=&MaskFloatBySpawnGroup(
        Masked Float=&RandomRangeFloat(Minimum=-337.5, Maximum=337.5)
        Passthrough Float=&RandomRangeFloat(Minimum=-150, Maximum=150)
       ))
     )
     Channel=::NewEnumerator1
    )
    Index 1 Param 3=&MakeFloatFromVector(
     Vector=&AddVectorToPosition(
      Position=$Engine.Owner.Position
      Vector=&MaskVectorBySpawnGroup(
       Masked Vector=&RandomRangeVector(Minimum=<0 0 900>, Maximum=<0 0 1000>)
       Passthrough Vector=&RandomRangeVector(Minimum=<0 0 600>, Maximum=<0 0 700>)
      )
     )
     Channel=::NewEnumerator2
    )
    Module.Index0Float0Write=true
    Module.Index0Float1Write=true
    Module.Index0Float2Write=true
    Module.Index0Float3Write=true
    Module.Index1Float0Write=true
    Module.Index1Float1Write=true
    Module.Index1Float2Write=true
    Module.Index1Float3Write=true
    Write Parameter Index 1=::NewEnumerator1
   )
   /SetVariables(
    Particles.Glow Color=&MultiplyLinearColor(
     Linear Color=<rgba 0.7119 0 1 1>
     Scale Factor=&MakeVector4FromVectorAndFloat(
      "Vector (RGB)"=<1 1 1>
      "Float (Alpha)"=&FloatFromCurve(FloatCurve=DataInterface() {propertyValues {Curve {
          defaultValue=1
          keys [
           {arriveTangent=9.0725, interpMode=RCIM_Cubic, time=0, value=0}
           {arriveTangent=0.2861, interpMode=RCIM_Cubic, leaveTangent=0.2861, tangentMode=RCTM_User, time=0.1, value=1}
           {arriveTangent=-3.5193, interpMode=RCIM_Cubic, leaveTangent=-3.5193, tangentMode=RCTM_User, time=1, value=0}
          ]
         }}})
     )
    )
    Particles.Glow Offset=-5
   ) {isSetParameters=true}
  ]
  rendererClasses [&NiagaraMeshRendererProperties]
  renderers [
   {rendererClass=&NiagaraMeshRendererProperties, rendererIndex=0}
   {rendererClass=&NiagaraMeshRendererProperties, rendererIndex=1}
   {rendererClass=&NiagaraMeshRendererProperties, rendererIndex=2}
  ]
 }
 WispySmokeFlare {
  emitterUpdateScript [
   &EmitterState()
   &SpawnBurst_Instantaneous(Spawn Count=5, Use Spawn Probability=false, UseLoopCountLimit=false)
  ]
  particleSpawnScript [
   &InitializeParticle(
    Color Maximum=<rgba 0.6104 0 0.7454 1>
    Color Minimum=<rgba 0.604 0 0.7454 0.7>
    Color Mode=::NewEnumerator2
    Lifetime Max=3
    Lifetime Min=2.2
    Lifetime Mode=::NewEnumerator1
    Sprite Rotation Mode=::NewEnumerator2
    Sprite Size Mode=::NewEnumerator1
    Sprite UV Mode=::NewEnumerator3
    Uniform Sprite Size Max=300
    Uniform Sprite Size Min=150
    UsePositionOffset=false
    Write Material Random=true
   )
   &ShapeLocation(
    Cylinder Height=120
    Cylinder Height Midpoint=0
    Cylinder Radius=1
    MaskSpawnGroup=false
    Offset=<0 0 -100>
    Offset Coordinate Space=::Local
    Offset Mode=::NewEnumerator0
    Override Seed=false
    Shape Primitive=::NewEnumerator1
    Surface Only=true
    UseOwnerScale=true
   )
   &AddVelocity(
    Velocity=$Particles.ShapeLocation.ShapeNormal
    Velocity Speed Scale=&RandomRangeFloat(Minimum=78, Maximum=240)
   )
   &AddVelocity(Velocity=&RandomRangeVector(Minimum=<0 0 25>, Maximum=<0 0 300>)) {module=AddVelocity001}
  ]
  particleUpdateScript [
   &ParticleState()
   &Drag(Drag=1.2, Use Linear Drag=true, Use Rotational Drag=true)
   &AccelerationForce(Acceleration=<0 0 50>)
   &SolveForcesAndVelocity(Clamp Velocity=false, Limit Acceleration=false)
   &SubUVAnimation(
    "Sprite Renderer "=DataInterface() {propertyValues {SpriteRenderer=&self:SmokeLoop_0.NiagaraSpriteRendererProperties_0}}
    Start Frame Range Override=&RandomRangeInt(Minimum=24, Maximum=32)
    UseEndFrame=false
    UseStartFrame=true
   )
   &DynamicMaterialParameters(
    Index 0 Param 1=&RandomRangeFloat(Minimum=0, Maximum=1)
    Index 0 Param 2=&RandomRangeFloat(Minimum=1, Maximum=2)
    Module.Index0Param0Write=true
    Module.Index0Param1Write=true
    Module.Index0Param2Write=true
    Module.Index0Param3Write=false
   )
   &SpriteRotationRate(Rotation Rate=&RandomRangeFloat(Minimum=-0.75, Maximum=-1))
   &ScaleColor(
    Scale Alpha=&FloatFromCurve(
     FloatCurve=DataInterface() {propertyValues {Curve {
        defaultValue=0
        keys [
         {arriveTangent=0.5571, interpMode=RCIM_Cubic, leaveTangent=0.5571, tangentMode=RCTM_User, time=0, value=0}
         {arriveTangent=-2.2607, interpMode=RCIM_Cubic, leaveTangent=-2.2607, tangentMode=RCTM_User, time=0.1408, value=0.9847}
         {arriveTangent=-0.1828, interpMode=RCIM_Cubic, leaveTangent=-0.1828, tangentMode=RCTM_User, time=1, value=0}
        ]
       }}}
     Scale Curve=&Lerp_Float(A=8, B=1, Alpha=$System.CameraFade)
    )
    ScaleA=true
    ScaleRGB=true
   )
  ]
  rendererClasses [&NiagaraSpriteRendererProperties]
  renderers [{rendererClass=&NiagaraSpriteRendererProperties}]
 }
 glow001 {
  emitterUpdateScript [
   &EmitterState()
   &SpawnBurst_Instantaneous(Use Spawn Probability=false, UseLoopCountLimit=false)
  ]
  particleSpawnScript [
   &InitializeParticle(
    Color=<rgba 0.0365 0 0.0325 0.5>
    Lifetime=2.5
    Sprite Size Mode=::NewEnumerator3
    Uniform Sprite Size=500
    UsePositionOffset=false
    Write Material Random=true
   )
   &CameraOffset(Camera Offset Amount=75)
   &SystemLocation(MaskSpawnGroup=false, Offset=<0 0 5>)
  ]
  particleUpdateScript [
   &ParticleState()
   &SpriteFacingAndAlignment(Sprite Alignment=<0 0 1>, Sprite Facing=<0 0 1>, WriteAlignment=true, WriteFacing=true)
   &DynamicMaterialParameters(
    Index 0 Param 1=2
    Index 0 Param 2=5
    Module.Index0Float0Write=true
    Module.Index0Float1Write=true
    Module.Index0Float2Write=true
    Module.Index0Float3Write=true
   )
   /SetVariables(
    Particles.GlowSpriteSize:Vector2f=<320 320>
    Particles.Long glow color:LinearColor=&MakeLinearColorFromVectorAndFloat(
     "Vector (RGB)"=&MakeVectorFromLinearColorRGB(Color=&MultiplyLinearColor(
       Linear Color=<rgba 0.6376 0.0357 0.7305 1>
       Scale Factor=&MakeVector4_FromFloat(Float=&FNTimeOfDayScalar(Night Scale=0.4, Morning Scale=1, Evening Scale=1))
      ))
     "Float (Alpha)"=&Multiply_Float(
      A=0.5
      B=&FloatFromCurve(FloatCurve=DataInterface() {propertyValues {Curve {keys [<at 0 1 RCIM_Cubic>, {arriveTangent=-1.25, interpMode=RCIM_Cubic, leaveTangent=-1.25, time=0.5911, value=0.5951}, <at 0.8 0>]}}})
     )
    )
   ) {isSetParameters=true}
   &FN_Light_Attributes(
    Color=<rgba 1.7479 0 2 1>
    Create Color Attribute=true
    Create Exponent Attribute=true
    Create Radius Attribute=true
    Exponent=4
    Radius=&FloatFromCurve(
     FloatCurve=DataInterface() {propertyValues {Curve {
        defaultValue=1
        keys [
         {interpMode=RCIM_Cubic, tangentMode=RCTM_User, time=0, value=1}
         {arriveTangent=-3.0595, interpMode=RCIM_Cubic, leaveTangent=-3.0595, tangentMode=RCTM_User, time=1, value=0}
        ]
       }}}
     Scale Curve=300
    )
    Scale Color=&Multiply_Float(
     A=&FloatFromCurve(FloatCurve=DataInterface() {propertyValues {Curve {keys [
          {interpMode=RCIM_Cubic, tangentMode=RCTM_User, time=0, value=1}
          {interpMode=RCIM_Cubic, tangentMode=RCTM_User, time=0.2, value=1}
          {tangentMode=RCTM_User, time=0.8, value=0}
         ]}}})
     B=&FloatFromCurve(
      FloatCurve=DataInterface() {propertyValues {Curve {
         defaultValue=1
         keys [
          {arriveTangent=9.0725, interpMode=RCIM_Cubic, leaveTangent=9.0725, tangentMode=RCTM_User, time=0, value=1}
          {arriveTangent=1.6364, interpMode=RCIM_Cubic, leaveTangent=1.6364, tangentMode=RCTM_User, time=0.1408, value=8}
          {arriveTangent=-0.0405, interpMode=RCIM_Cubic, leaveTangent=-0.0405, tangentMode=RCTM_User, time=1, value=0}
         ]
        }}}
      Scale Curve=8
     )
    )
   )
   &ScaleColor(
    Scale Alpha=&FloatFromCurve(FloatCurve=DataInterface() {propertyValues {Curve {
        defaultValue=1
        keys [
         {arriveTangent=-0.2812, interpMode=RCIM_Cubic, leaveTangent=-0.2812, tangentMode=RCTM_User, time=0.4, value=1}
         {arriveTangent=-0.3043, interpMode=RCIM_Cubic, leaveTangent=-0.3043, tangentMode=RCTM_User, time=0.6, value=0}
        ]
       }}})
    ScaleA=true
    ScaleRGB=true
   )
  ]
  rendererClasses [&NiagaraSpriteRendererProperties, &NiagaraLightRendererProperties]
  renderers [
   {rendererClass=&NiagaraSpriteRendererProperties}
   {rendererClass=&NiagaraLightRendererProperties, rendererIndex=1}
   {rendererClass=&NiagaraSpriteRendererProperties, rendererIndex=2}
  ]
 }
}
summary {
 systemName=NS_BuzzCool_Defeated
 emitters [
  {emitterName=glow001, rendererClasses [&NiagaraSpriteRendererProperties, &NiagaraLightRendererProperties]}
  {emitterName=Flare002, rendererClasses [&NiagaraSpriteRendererProperties]}
  {emitterName=NE_Flare, rendererClasses [&NiagaraMeshRendererProperties, &NiagaraSpriteRendererProperties]}
  {emitterName=Embers, rendererClasses [&NiagaraSpriteRendererProperties]}
  {bEnabled=false, emitterName=Smoke_Masked_Donut, rendererClasses [&NiagaraMeshRendererProperties]}
  {
   bEnabled=false
   emitterName=Smoke_Masked_Donut002
   rendererClasses [&NiagaraMeshRendererProperties, &NiagaraSpriteRendererProperties]
  }
  {
   emitterName=Smoke_Masked_Donut001
   rendererClasses [&NiagaraMeshRendererProperties, &NiagaraSpriteRendererProperties]
  }
  {emitterName=WispySmokeFlare, rendererClasses [&NiagaraSpriteRendererProperties]}
  {emitterName=Spline_Meshes, rendererClasses [&NiagaraMeshRendererProperties]}
 ]
 userVariables []
}
