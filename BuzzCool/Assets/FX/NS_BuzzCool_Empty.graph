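# NS_BuzzCool_Empty is a Niagara system that draws nothing, apparently an empty placeholder effect.
# Its single emitter, named Light, burst-spawns particles, initializes them with a fixed lifetime and only updates their particle state.
# It has no renderers, despite the emitter name, and exposes no user parameters.
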NiagaraSystem /BuzzCool/Assets/FX/NS_BuzzCool_Empty
dependencies {usedDataInterfaces [], usedDynamicInputs [], usedModules [&SystemState]}
emitters {Light {
  emitterUpdateScript [
   &EmitterState()
   &SpawnBurst_Instantaneous(Use Spawn Probability=false, UseLoopCountLimit=true)
  ]
  particleSpawnScript [&InitializeParticle(Color Mode=::NewEnumerator0, Lifetime=5, UsePositionOffset=false, Write Material Random=true)]
  particleUpdateScript [&ParticleState()]
  rendererClasses []
  renderers []
 }}
summary {systemName=NS_BuzzCool_Empty, emitters [{emitterName=Light, rendererClasses []}], userVariables []}
