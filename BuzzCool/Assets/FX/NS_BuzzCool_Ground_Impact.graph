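# Niagara burst effect for a BuzzCool ground impact: shockwave and decal meshes, dark streak sprites, speed lines and two mesh smoke layers.
# Shockwaves_and_Decal drives several mesh renderers from one emitter, using per-mesh attributes and a spawn-group VisibilityTag.
# User.Size scales sizes and some velocities, while time-of-day functions tint smoke and streaks and feed shockwave material parameters.
# User.Frosting and User.Golden are declared, but no shown module reads them.
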NiagaraSystem /BuzzCool/Assets/FX/NS_BuzzCool_Ground_Impact
dependencies {
 usedDataInterfaces [&NiagaraDataInterfaceColorCurve, &NiagaraDataInterfaceCurve, &NiagaraDataInterfaceArrayFloat3]
 usedDynamicInputs []
 usedModules [&SystemState]
}
emitters {
 DarkStreaks {
  emitterUpdateScript [
   &EmitterState()
   &SpawnBurst_Instantaneous(
    Spawn Count=&RandomRangeInt(Minimum=6, Maximum=11)
    Use Spawn Probability=false
    UseLoopCountLimit=false
   )
  ]
  particleSpawnScript [
   /SetVariables(Particles.RandomFloat=&UniformRangedFloat(Minimum=0, Maximum=3)) {isSetParameters=true}
   &InitializeParticle(
    Color Maximum=<rgba 0.8021 0.8021 0.8021 1>
    Color Minimum=<rgba 0.9271 0.9271 0.9271 1>
    Color Mode=::NewEnumerator2
    Lifetime Max=0.5
    Lifetime Min=0.3
    Lifetime Mode=::NewEnumerator1
    Sprite Size Max=<75 495>
    Sprite Size Min=<67.5 360>
    Sprite Size Mode=::NewEnumerator2
    UsePositionOffset=false
    Write Material Random=true
   )
   &ShapeLocation(
    MaskSpawnGroup=false
    Override Seed=false
    Ring Radius=&Lerp_Float(A=100, B=300, Alpha=$User.Size)
    Rotation Axis=<0 0 1>
    Rotation Mode=::NewEnumerator1
    Shape Primitive=::NewEnumerator4
    UseOwnerScale=true
   )
   &AddVelocity(
    Module.Use Falloff Distance=false
    Override Seed=false
    Velocity Mode=::NewEnumerator1
    Velocity Speed=&Lerp_Float(
     A=87.5
     B=125
     Alpha=&FloatFromCurve(
      FloatCurve=DataInterface() {propertyValues {Curve {keys [{interpMode=RCIM_Cubic, tangentMode=RCTM_Break, time=0, value=0}, {interpMode=RCIM_Cubic, tangentMode=RCTM_Break, time=1, value=1}]}}}
      CurveIndex=&RandomRangeFloat(Minimum=0, Maximum=1)
     )
    )
   )
   &AddVelocity(Velocity=&RandomRangeVector(Minimum=<0 0 100>, Maximum=<0 0 150>)) {module=AddVelocity001}
   &SubUVAnimation(End Frame=3, SubUV Animation Mode:ENiagara_SubUVLookupMode=::NewEnumerator2)
  ]
  particleUpdateScript [
   &ParticleState()
   &Drag(Drag=4, Use Linear Drag=true, Use Rotational Drag=false)
   &SolveForcesAndVelocity(Clamp Velocity=false, Limit Acceleration=false)
   &ScaleSpriteSize(Scale Factor=&Vector2DFromFloat(Value=&FloatFromCurve(
      FloatCurve=DataInterface() {propertyValues {Curve {keys [
          {arriveTangent=1.6317, interpMode=RCIM_Cubic, leaveTangent=1.6317, tangentMode=RCTM_User, time=0, value=0.1}
          {arriveTangent=0.2754, interpMode=RCIM_Cubic, leaveTangent=0.2754, tangentMode=RCTM_User, time=1.0463, value=1}
         ]}}}
      Scale Curve=&Lerp_Float(A=2, B=4, Alpha=$User.Size)
     )))
   &ScaleColor(
    Scale Alpha=&FloatFromCurve(
     FloatCurve=DataInterface() {propertyValues {Curve {keys [{interpMode=RCIM_Cubic, tangentMode=RCTM_Break, time=0, value=1}, {interpMode=RCIM_Cubic, tangentMode=RCTM_Break, time=1, value=0}]}}}
     Scale Curve=&RandomRangeFloat(Minimum=1, Maximum=1.25)
    )
    Scale RGB=&MakeVectorFromLinearColorRGB(Color=&FNTimeOfDayColor(
      Night Color=<rgba 0.3804 0.429 0.5938 1>
      Morning Color=<rgba 0.6181 0.5423 0.8073 1>
      Day Color=<rgba 0.6517 0.919 1 1>
      Evening Color=<rgba 0.8073 0.5874 0.6207 1>
     ))
    ScaleA=true
    ScaleRGB=true
   )
   &DynamicMaterialParameters(
    Index 0 Param 2=1
    Index 0 Param 3=1
    Index 0 Param 4=&FloatFromCurve(FloatCurve=DataInterface() {propertyValues {Curve {keys [<at 0.3 1>,<at 1 0.25>]}}})
    Module.Param0WriteEnabled=true
    Module.Param1WriteEnabled=true
    Module.Param2WriteEnabled=true
    Module.Param3WriteEnabled=true
   )
  ]
  rendererClasses [&NiagaraSpriteRendererProperties]
  renderers [{rendererClass=&NiagaraSpriteRendererProperties}]
 }
 Shockwaves_and_Decal {
  emitterUpdateScript [
   &EmitterState()
   &SpawnBurst_Instantaneous(Spawn Count=2, Spawn Group=2, Use Spawn Probability=false, UseLoopCountLimit=false) {module=SpawnBurst_Instantaneous001}
   &SpawnBurst_Instantaneous(Spawn Count=3, Spawn Group=3, Use Spawn Probability=false, UseLoopCountLimit=false) {module=SpawnBurst_Instantaneous002}
   &SpawnBurst_Instantaneous(
    Spawn Count=&RandomRangeInt(Minimum=2, Maximum=4)
    Spawn Group=4
    Use Spawn Probability=false
    UseLoopCountLimit=false
   ) {module=SpawnBurst_Instantaneous003}
  ]
  particleSpawnScript [
   /SetVariables(Particles.SpawnGroup=$Emitter.SpawnGroup) {isSetParameters=true}
   &InitializeParticle(Color Mode=::NewEnumerator0, Lifetime=0.8, UsePositionOffset=false, Write Material Random=true)
   /SetVariables(
    Particles.Mesh_02_Color:LinearColor=<rgba 0.807 0.8723 1 0.15>
    Particles.Mesh_02_Orientation:Quat4f=&MakeQuaternion(Quaternion From=0, Axis=<0 1 0>, Angle=0.25, From=?, To=?, Angle Type=::NewEnumerator1, Normalize Vectors=false)
    Particles.Mesh_02_Size:Vector3f=&VectorFromFloat(Value=&Lerp_Float(A=18, B=27, Alpha=$User.Size))
   ) {isSetParameters=true}
   /SetVariables(
    Particles.Mesh_03_Color:LinearColor=<rgba 1 1 1 50>
    Particles.Mesh_03_Orientation:Quat4f=&MultiplyQuaternion(
     Quaternion A=&MakeQuaternion(Quaternion From=0, Axis=<0 0 1>, Angle=0, From=?, To=?, Angle Type=::NewEnumerator1, Coordinate Space=::World, Normalize Vectors=false)
     Quaternion B=&MakeQuaternion(
      Quaternion From=0
      Axis=<0 0 1>
      Angle=&RandomRangeFloat(Minimum=-1, Maximum=1)
      From=?
      To=?
      Angle Type=::NewEnumerator1
      Normalize Vectors=false
     )
    )
    Particles.Mesh_03_Size:Vector3f=&VectorFromFloat(Value=&Lerp_Float(A=4, B=7.5, Alpha=$User.Size))
   ) {isSetParameters=true}
   /SetVariables(
    Particles.Mesh_04_Orientation:Quat4f=&MultiplyQuaternion(
     Quaternion A=&MakeQuaternion(Quaternion From=0, Axis=<0 0 1>, Angle=0, From=?, To=?, Angle Type=::NewEnumerator1, Coordinate Space=::World, Normalize Vectors=false)
     Quaternion B=&MakeQuaternion(
      Quaternion From=0
      Axis=<0 0 1>
      Angle=&RandomRangeFloat(Minimum=-1, Maximum=1)
      From=?
      To=?
      Angle Type=::NewEnumerator1
      Normalize Vectors=false
     )
    )
    Particles.Mesh_04_Size:Vector3f=&VectorFromFloat(Value=&Multiply_Float(
      A=&RandomRangeFloat(Minimum=2.2, Maximum=2.8)
      B=&Lerp_Float(A=1.5, B=3, Alpha=$User.Size)
     ))
   ) {isSetParameters=true}
   /SetVariables(Particles.RandomRange:NiagaraFloat=&RandomRangeFloat(Minimum=0.5, Maximum=0.75)) {isSetParameters=true}
   &CameraOffset(Camera Offset Amount=50)
  ]
  particleUpdateScript [
   &ParticleState()
   /SetVariables(
    Particles.Mesh_02_Color:LinearColor=&MultiplyLinearColor(
     Linear Color=$Particles.Initial.Mesh_02_Color
     Scale Factor=&MakeFloat4FromVectorAndFloat(
      "Vector (RGB)"=&MakeVectorFromLinearColorRGB(Color=&ColorFromCurve(
        ColorCurve=DataInterface() {propertyValues {
          AlphaCurve {keys [<at 0 1 RCIM_Cubic>]}
          BlueCurve {keys [<at 0.0388 3 RCIM_Cubic>,<at 0.9534 1>]}
          GreenCurve {keys [<at 0.0388 3 RCIM_Cubic>,<at 0.9534 0.8008>]}
          RedCurve {keys [<at 0.0388 3 RCIM_Cubic>,<at 0.9534 0.4807>]}
         }}
        CurveIndex=&Divide_Float(A=$Particles.NormalizedAge, B=0.5)
       ))
      "Float (Alpha)"=&FloatFromCurve(
       FloatCurve=DataInterface() {propertyValues {Curve {keys [
           {tangentMode=RCTM_User, time=0, value=1}
           {arriveTangent=-0.0405, interpMode=RCIM_Cubic, leaveTangent=-0.0405, tangentMode=RCTM_User, time=0.5, value=0.5}
           {arriveTangent=-2.4747, interpMode=RCIM_Cubic, leaveTangent=-2.4747, tangentMode=RCTM_User, time=1, value=0}
          ]}}}
       CurveIndex=&Divide_Float(A=$Particles.NormalizedAge, B=0.5)
       Scale Curve=2
      )
     )
    )
    Particles.Mesh_02_DynamicParameter:Vector4f=&MakeVector4(
     X=1
     Y=&FloatFromCurve(
      FloatCurve=DataInterface() {propertyValues {Curve {keys [
          {interpMode=RCIM_Cubic, tangentMode=RCTM_User, time=0, value=1}
          {arriveTangent=-1, interpMode=RCIM_Cubic, leaveTangent=-1, time=0.6, value=0.7}
          {interpMode=RCIM_Cubic, leaveTangent=-3.3753, time=1, value=0}
         ]}}}
      CurveIndex=&Divide_Float(A=$Particles.NormalizedAge, B=0.5)
     )
     Z=0.5
     W=200
    )
    Particles.Mesh_02_Size:Vector3f=&Multiply_Vector(
     A=$Particles.Initial.Mesh_02_Size
     B=&VectorFromFloat(Value=&FloatFromCurve(FloatCurve=DataInterface() {propertyValues {Curve {keys [
           {arriveTangent=6.728, interpMode=RCIM_Cubic, leaveTangent=6.728, tangentMode=RCTM_User, time=0, value=0.05}
           {arriveTangent=1.2204, interpMode=RCIM_Cubic, leaveTangent=1.2204, tangentMode=RCTM_User, time=0.15, value=0.6825}
           {arriveTangent=-0.0405, interpMode=RCIM_Cubic, leaveTangent=-0.0405, tangentMode=RCTM_User, time=1, value=1}
          ]}}}))
    )
   ) {isSetParameters=true}
   /SetVariables(
    Particles.Mesh_03_Color:LinearColor=&MultiplyLinearColor(
     Linear Color=<rgba 1 1 1 0.6>
     Scale Factor=&MakeFloat4FromVectorAndFloat(
      "Vector (RGB)"=&MakeVectorFromLinearColorRGB(Color=&ColorFromCurve(
        ColorCurve=DataInterface() {propertyValues {
          AlphaCurve {keys [<at 0 1 RCIM_Cubic>]}
          BlueCurve {keys [<at 0.043 1 RCIM_Cubic>,<at 1 0.3281>]}
          GreenCurve {keys [<at 0.043 1 RCIM_Cubic>,<at 1 0.3281>]}
          RedCurve {keys [<at 0.043 1 RCIM_Cubic>,<at 1 0.3281>]}
         }}
        CurveIndex=&Divide_Float(A=$Particles.NormalizedAge, B=0.5)
       ))
      "Float (Alpha)"=&FloatFromCurve(
       FloatCurve=DataInterface() {propertyValues {Curve {keys [
           {arriveTangent=10.8551, interpMode=RCIM_Cubic, leaveTangent=10.8551, tangentMode=RCTM_User, time=0, value=0.5}
           {arriveTangent=-0.7357, interpMode=RCIM_Cubic, leaveTangent=-0.7357, tangentMode=RCTM_User, time=0.1, value=1}
           {arriveTangent=-1.0038, interpMode=RCIM_Cubic, leaveTangent=-1.0038, tangentMode=RCTM_User, time=0.4, value=0.2}
           {arriveTangent=0.4048, interpMode=RCIM_Cubic, leaveTangent=0.4048, tangentMode=RCTM_User, time=0.7, value=0}
          ]}}}
       CurveIndex=&Divide_Float(A=$Particles.NormalizedAge, B=0.7)
      )
     )
    )
    Particles.Mesh_03_DynamicParameter:Vector4f=&MakeVector4(
     X=0
     Y=&FNTimeOfDayScalar(Night Scale=0.05, Morning Scale=-0.26, Day Scale=0, Evening Scale=-0.35)
     Z=&FNTimeOfDayScalar(Night Scale=0.5, Morning Scale=0.23, Evening Scale=0.2)
     W=&FNTimeOfDayScalar(Night Scale=0.02, Morning Scale=-0.01, Day Scale=0, Evening Scale=-0.01)
    )
    Particles.Mesh_03_Size:Vector3f=&Multiply_Vector(
     A=$Particles.Initial.Mesh_03_Size
     B=&Multiply_VectorByFloat(
      Vector=&MakeVector(
       X=1
       Y=1
       Z=&FloatFromCurve(
        FloatCurve=DataInterface() {propertyValues {Curve {keys [<at 0 0.7>,<at 1 0>]}}}
        CurveIndex=&Divide_Float(A=$Particles.NormalizedAge, B=0.5)
       )
      )
      Float=&FloatFromCurve(
       FloatCurve=DataInterface() {propertyValues {Curve {keys [
           {arriveTangent=12.6277, interpMode=RCIM_Cubic, leaveTangent=12.6277, tangentMode=RCTM_User, time=0, value=0}
           {arriveTangent=1.6578, interpMode=RCIM_Cubic, leaveTangent=1.6578, tangentMode=RCTM_User, time=0.3184, value=1.6506}
           {arriveTangent=-0.0405, interpMode=RCIM_Cubic, leaveTangent=-0.0405, tangentMode=RCTM_User, time=1.0024, value=2}
          ]}}}
       CurveIndex=&Divide_Float(A=$Particles.NormalizedAge, B=0.5)
      )
     )
    )
   ) {isSetParameters=true}
   /SetVariables(
    Particles.Mesh_04_Color:LinearColor=&MultiplyLinearColor(
     Linear Color=$Particles.Initial.Mesh_04_Color
     Scale Factor=&MakeFloat4FromVectorAndFloat(
      "Vector (RGB)"=&MakeVectorFromLinearColorRGB(Color=&ColorFromCurve(
        ColorCurve=DataInterface() {propertyValues {
          AlphaCurve {keys [<at 0 1 RCIM_Cubic>]}
          BlueCurve {keys [<at 0.077 1 RCIM_Cubic>,<at 0.9834 0.4479>]}
          GreenCurve {keys [<at 0.077 1 RCIM_Cubic>,<at 0.9834 0.3755>]}
          RedCurve {keys [<at 0.077 1 RCIM_Cubic>,<at 0.9834 0.2415>]}
         }}
        CurveIndex=&Divide_Float(A=$Particles.NormalizedAge, B=&RandomRangeFloat(Minimum=0.5, Maximum=0.75))
       ))
      "Float (Alpha)"=&FloatFromCurve(
       FloatCurve=DataInterface() {propertyValues {Curve {keys [
           {arriveTangent=8.2356, interpMode=RCIM_Cubic, leaveTangent=8.2356, tangentMode=RCTM_User, time=0, value=0}
           {arriveTangent=0.1969, interpMode=RCIM_Cubic, leaveTangent=0.1969, tangentMode=RCTM_User, time=0.2223, value=1}
           {arriveTangent=-3.0486, interpMode=RCIM_Cubic, leaveTangent=-3.0486, tangentMode=RCTM_User, time=0.4781, value=0.509}
           {arriveTangent=-0.2378, interpMode=RCIM_Cubic, leaveTangent=-0.2378, tangentMode=RCTM_User, time=1, value=0}
          ]}}}
       CurveIndex=&Divide_Float(A=$Particles.NormalizedAge, B=$Particles.RandomRange)
       Scale Curve=0.6
      )
     )
    )
    Particles.Mesh_04_DynamicParameter:Vector4f=&MakeVector4(
     X=&FloatFromCurve(
      FloatCurve=DataInterface() {propertyValues {Curve {keys [{interpMode=RCIM_Cubic, tangentMode=RCTM_User, time=0.5, value=0}, {interpMode=RCIM_Cubic, tangentMode=RCTM_User, time=1, value=1}]}}}
      CurveIndex=&Divide_Float(A=$Particles.NormalizedAge, B=$Particles.RandomRange)
     )
     Y=&FNTimeOfDayScalar(Night Scale=0.05, Morning Scale=-0.26, Day Scale=0, Evening Scale=-0.35)
     Z=&FNTimeOfDayScalar(Night Scale=0.5, Morning Scale=0.23, Evening Scale=0.2)
     W=&FNTimeOfDayScalar(Night Scale=0.02, Morning Scale=-0.01, Day Scale=0, Evening Scale=-0.01)
    )
    Particles.Mesh_04_Orientation:Quat4f=&MultiplyQuaternion(
     Quaternion A=$Particles.Initial.Mesh_04_Orientation
     Quaternion B=&MakeQuatFromAxisAngle(
      Axis=&RandomRangeVector(Minimum=<0 0 -1>, Maximum=<0 0 1>)
      AngleInDegrees=&RandomRangeFloat(Minimum=-50, Maximum=50)
     )
    )
    Particles.Mesh_04_Position:NiagaraPosition=&Add_Vector(A=Linked() {linkedVariable {name=Particles.Position, type=NiagaraPosition}}, B=<25 0 0>)
    Particles.Mesh_04_Size:Vector3f=&Multiply_Vector(
     A=$Particles.Initial.Mesh_04_Size
     B=&Multiply_VectorByFloat(
      Vector=&MakeVector(
       X=1
       Y=1
       Z=&FloatFromCurve(
        FloatCurve=DataInterface() {propertyValues {Curve {keys [<at 0 0.7>,<at 1 0>]}}}
        CurveIndex=&Divide_Float(A=$Particles.NormalizedAge, B=0.5)
       )
      )
      Float=&FloatFromCurve(
       FloatCurve=DataInterface() {propertyValues {Curve {keys [
           {arriveTangent=12.6277, interpMode=RCIM_Cubic, leaveTangent=12.6277, tangentMode=RCTM_User, time=0, value=0}
           {arriveTangent=1.6578, interpMode=RCIM_Cubic, leaveTangent=1.6578, tangentMode=RCTM_User, time=0.3184, value=1.6506}
           {arriveTangent=-0.0405, interpMode=RCIM_Cubic, leaveTangent=-0.0405, tangentMode=RCTM_User, time=1.0024, value=2}
          ]}}}
       CurveIndex=&Divide_Float(A=$Particles.NormalizedAge, B=0.5)
      )
     )
    )
   ) {isSetParameters=true}
   /SetVariables(
    Particles.Mesh_05_Color:LinearColor=/Niagara/DynamicInputs/LinearColor/MakeLinearColorFromVector4(Vector:Vector4f=&MakeFloat4FromVectorAndFloat(
      "Vector (RGB)"=&MakeVectorFromLinearColorRGB(Color=&ColorFromCurve(
        ColorCurve=DataInterface() {propertyValues {
          AlphaCurve {keys [<at 0 1 RCIM_Cubic>,<at 0.9854 1>]}
          BlueCurve {keys [<at 0 0.6719>,<at 0.6508 0.2448 RCIM_Cubic>]}
          GreenCurve {keys [<at 0 0.6539>,<at 0.6508 0.2331 RCIM_Cubic>]}
          RedCurve {keys [<at 0 0.4614>,<at 0.6508 0.1813 RCIM_Cubic>]}
         }}
        CurveIndex=&Divide_Float(A=$Particles.NormalizedAge, B=0.5)
       ))
      "Float (Alpha)"=0.07
     ))
    Particles.Mesh_05_DynamicParameter:Vector4f=&MakeVector4(
     X=&FloatFromCurve(
      FloatCurve=DataInterface() {propertyValues {Curve {keys [
          {arriveTangent=4.1785, interpMode=RCIM_Cubic, leaveTangent=4.1785, tangentMode=RCTM_User, time=0, value=0}
          {arriveTangent=1.1179, interpMode=RCIM_Cubic, leaveTangent=1.1179, tangentMode=RCTM_User, time=0.2, value=0.6228}
          {arriveTangent=0.0224, interpMode=RCIM_Cubic, leaveTangent=0.0224, tangentMode=RCTM_User, time=1, value=1}
         ]}}}
      CurveIndex=&Divide_Float(A=$Particles.NormalizedAge, B=0.5)
     )
     Y=0
     Z=&Lerp_Float(
      A=0.85
      B=1
      Alpha=&FloatFromCurve(
       FloatCurve=DataInterface() {propertyValues {Curve {keys [<at 0 0 RCIM_Cubic>, {arriveTangent=2.7526, interpMode=RCIM_Cubic, leaveTangent=2.7526, tangentMode=RCTM_User, time=1, value=1}]}}}
       CurveIndex=&Divide_Float(A=$Particles.NormalizedAge, B=0.5)
      )
     )
     W=0
    )
    Particles.Mesh_05_Orientation:Quat4f=&MakeQuaternion(
     Quaternion From=3
     Axis=<1 0 0>
     Angle=0
     From=<1 0 0>
     To=<1 0 0>
     Angle Type=::NewEnumerator0
     Coordinate Space=::World
     Normalize Vectors=false
    )
    Particles.Mesh_05_Size:Vector3f=&Multiply_Vector(
     A=$Particles.Initial.Mesh_02_Size
     B=&VectorFromFloat(Value=&FloatFromCurve(
       FloatCurve=DataInterface() {propertyValues {Curve {keys [
           {arriveTangent=6.728, interpMode=RCIM_Cubic, leaveTangent=6.728, tangentMode=RCTM_User, time=0, value=0.05}
           {arriveTangent=1.2204, interpMode=RCIM_Cubic, leaveTangent=1.2204, tangentMode=RCTM_User, time=0.15, value=0.6825}
           {arriveTangent=-0.0405, interpMode=RCIM_Cubic, leaveTangent=-0.0405, tangentMode=RCTM_User, time=1, value=1}
          ]}}}
       Scale Curve=&Lerp_Float(A=0.9, B=1.2, Alpha=$User.Size)
      ))
    )
   ) {isSetParameters=true}
   /SetVariables(Particles.Mesh_Position:NiagaraPosition=&Add_Vector(
     A=Linked() {linkedVariable {name=Particles.Position, type=NiagaraPosition}}
     B=&RandomRangeVector(Minimum=<2 0 0>, Maximum=<5 0 0>)
    )) {isSetParameters=true}
   /SetVariables(Particles.VisibilityTag=$Particles.SpawnGroup) {isSetParameters=true}
  ]
  rendererClasses [&NiagaraMeshRendererProperties]
  renderers [
   {rendererClass=&NiagaraMeshRendererProperties, rendererIndex=0}
   {rendererClass=&NiagaraMeshRendererProperties, rendererIndex=1}
   {rendererClass=&NiagaraMeshRendererProperties, rendererIndex=2}
   {rendererClass=&NiagaraMeshRendererProperties, rendererIndex=3}
  ]
 }
 Smoke_Circular {
  emitterUpdateScript [
   &EmitterState()
   &SpawnBurst_Instantaneous(Spawn Count=4, Use Spawn Probability=false, UseLoopCountLimit=false)
  ]
  particleSpawnScript [
   &InitializeParticle(
    Lifetime Max=0.9
    Lifetime Min=0.7
    Lifetime Mode=::NewEnumerator1
    Mesh Scale Mode=::NewEnumerator1
    Mesh Uniform Scale Max=20
    Mesh Uniform Scale Min=20
    UsePositionOffset=false
    Write Material Random=true
   )
   &InitialMeshOrientation(
    Mesh Axis To Orient=&SelectVectorFromArray(
     Direct Array Index=&ReturnExecIndex()
     Vector Selection Array=DataInterface() {propertyValues {FloatData [<1 1 0>, <-1 -1 0>, <1 -1 0>, <-1 1 0>]}}
    )
    Mesh Orientation Mode=::NewEnumerator2
    Use Rotation Vector=false
   )
   &CameraOffset(Camera Offset Amount=100)
   &AddVelocity(
    Velocity=&SelectVectorFromArray(
     Direct Array Index=&ReturnExecIndex()
     Vector Selection Array=DataInterface() {propertyValues {FloatData [<0 -1 0>, <0 1 0>, <1 0 0>, <-1 0 0>]}}
    )
    Velocity Speed Scale=1000
   )
  ]
  particleUpdateScript [
   &ParticleState()
   &Drag(Drag=2, Use Linear Drag=true, Use Rotational Drag=true)
   &SolveForcesAndVelocity(Clamp Velocity=false, Limit Acceleration=false)
   &ScaleMeshSize(Scale Factor=&VectorFromFloat(Value=&FloatFromCurve(
      FloatCurve=DataInterface() {propertyValues {Curve {
         defaultValue=1
         keys [
          {arriveTangent=2.9875, interpMode=RCIM_Cubic, leaveTangent=2.9875, tangentMode=RCTM_User, time=0, value=0}
          {interpMode=RCIM_Cubic, leaveTangent=0.9093, time=1, value=1}
         ]
        }}}
      Scale Curve=&Lerp_Float(A=1, B=1.8, Alpha=$User.Size)
     )))
   &FN_Mesh_Attributes(
    Color=&MultiplyLinearColor(
     Linear Color=<rgba 1 1 1 1>
     Scale Factor=&MakeVector4FromVectorAndFloat(
      "Vector (RGB)"=&MakeVectorFromLinearColorRGB(Color=&FNTimeOfDayColor(
        Night Color=<rgba 0.2614 0.3686 0.401 1>
        Morning Color=<rgba 0.6181 0.5423 0.8073 1>
        Day Color=<rgba 0.6517 0.919 1 1>
        Evening Color=<rgba 0.8073 0.5874 0.6207 1>
       ))
      "Float (Alpha)"=&FloatFromCurve(
       FloatCurve=DataInterface() {propertyValues {Curve {
          defaultValue=0
          keys [
           {arriveTangent=-0.9317, interpMode=RCIM_Cubic, leaveTangent=-0.9317, tangentMode=RCTM_User, time=0.4, value=1}
           {arriveTangent=-1.4312, interpMode=RCIM_Cubic, leaveTangent=-1.4312, tangentMode=RCTM_User, time=1, value=0}
          ]
         }}}
       Scale Curve=0.8
      )
     )
    )
    Create Color Attribute=true
    Create Scale Attribute=true
    Scale=&Multiply_Vector(A=$Particles.Scale, B=<1 1 1>)
    Write to Module Namespace=true
   )
   &ScaleColor(
    Scale Alpha=&FloatFromCurve(FloatCurve=DataInterface() {propertyValues {Curve {keys [
         {arriveTangent=0.005, interpMode=RCIM_Cubic, leaveTangent=0.005, tangentMode=RCTM_User, time=0, value=0.7}
         {arriveTangent=-0.6104, leaveTangent=0.3906, tangentMode=RCTM_User, time=0.7, value=0.4172}
        ]}}})
    Scale RGB=&MakeVectorFromLinearColorRGB(Color=&FNTimeOfDayColor(
      Night Color=<rgba 0.2614 0.3686 0.401 1>
      Morning Color=<rgba 0.6181 0.5423 0.8073 1>
      Day Color=<rgba 0.6517 0.919 1 1>
      Evening Color=<rgba 0.8073 0.5874 0.6207 1>
     ))
    ScaleA=true
    ScaleRGB=true
   )
  ]
  rendererClasses [&NiagaraMeshRendererProperties]
  renderers [
   {rendererClass=&NiagaraMeshRendererProperties, rendererIndex=0}
   {rendererClass=&NiagaraMeshRendererProperties, rendererIndex=1}
  ]
 }
 Smoke_Masked002 {
  emitterUpdateScript [
   &EmitterState()
   &SpawnBurst_Instantaneous(Spawn Count=14, Use Spawn Probability=false, UseLoopCountLimit=false)
   &SpawnBurst_Instantaneous(Spawn Count=5, Spawn Group=1, Spawn Time=0.001, Use Spawn Probability=true, UseLoopCountLimit=false) {module=SpawnBurst_Instantaneous001}
  ]
  particleSpawnScript [
   &InitializeParticle(
    Lifetime Max=1.4
    Lifetime Min=0.9
    Lifetime Mode=::NewEnumerator1
    Mesh Scale Mode=::NewEnumerator1
    Mesh Uniform Scale Max=12
    Mesh Uniform Scale Min=11
    UsePositionOffset=false
    Write Material Random=true
   )
   &ShapeLocation(
    Cylinder Distribution=::NewEnumerator2
    Cylinder Height=10
    Cylinder Height Midpoint=0
    Cylinder Radius=60
    MaskSpawnGroup=true
    Offset Coordinate Space=::World
    Offset Mode=::NewEnumerator0
    Rotation Axis=<0 0 1>
    Rotation Mode=::NewEnumerator1
    Shape Primitive=::NewEnumerator1
    U Position=&ReturnNormalizedExecIndex(Normalized Index Includes One=false)
    UseOwnerScale=true
   )
   &ShapeLocation(
    MaskSpawnGroup=true
    Offset Coordinate Space=::World
    Offset Mode=::NewEnumerator0
    Rotation Axis=<0 0 1>
    Rotation Mode=::NewEnumerator1
    Spawn Group Mask=1
    Sphere Distribution=::NewEnumerator1
    Sphere Radius=20
    U Position=&ReturnNormalizedExecIndex(Normalized Index Includes One=false)
    UseOwnerScale=true
    V Position=0.7
   ) {module=ShapeLocation001}
   &AddVelocity(
    Velocity=&Lerp_Vector(
     A=$Particles.ShapeLocation.ShapeVector
     B=&Add_Vector(
      A=$Particles.ShapeLocation001.ShapeVector
      B=&Add_Vector(
       A=&Multiply_VectorByFloat(Vector=$Particles.ShapeLocation.ShapeVector, Float=0.5)
       B=&Lerp_Vector(
        A=&RandomRangeVector(Minimum=<0 0 11.2>, Maximum=<0 0 24>)
        B=&RandomRangeVector(Minimum=<0 0 4>, Maximum=<0 0 8>)
        Alpha=$User.Size
       )
      )
     )
     Alpha=&MakeFloatFromInt(Int=$Emitter.SpawnGroup)
    )
    Velocity Speed Scale=&Lerp_Float(A=15, B=36, Alpha=$User.Size)
   )
   &InitialMeshOrientation(Use Rotation Vector=false)
   &CameraOffset(Camera Offset Amount=100)
  ]
  particleUpdateScript [
   &ParticleState()
   &Drag(Drag=&Lerp_Float(A=1.5, B=2.25, Alpha=$User.Size), Use Linear Drag=true, Use Rotational Drag=true)
   &SolveForcesAndVelocity(Clamp Velocity=false, Limit Acceleration=false)
   &ScaleMeshSize(Scale Factor=&Multiply_VectorByFloat(
     Vector=<1 1 1>
     Float=&FloatFromCurve(
      FloatCurve=DataInterface() {propertyValues {Curve {keys [
          <at 0 1>
          {arriveTangent=-0.8765, interpMode=RCIM_Cubic, leaveTangent=-0.8765, tangentMode=RCTM_User, time=0.3, value=1}
          {arriveTangent=-2.3538, interpMode=RCIM_Cubic, leaveTangent=-2.3538, tangentMode=RCTM_User, time=0.8, value=0}
         ]}}}
      Scale Curve=2
     )
    ))
   &UpdateMeshOrientation(Rotation Rate=0.4, Rotation Vector=<0 1 0>)
   &ScaleColor(
    Scale Alpha=&FloatFromCurve(
     FloatCurve=DataInterface() {propertyValues {Curve {keys [
         {arriveTangent=-0.2008, interpMode=RCIM_Cubic, leaveTangent=-0.2008, tangentMode=RCTM_User, time=0.4, value=0.5}
         {interpMode=RCIM_Cubic, leaveTangent=0.3906, time=0.8, value=0.4}
        ]}}}
     Scale Curve=0.9
    )
    Scale RGB=&MakeVectorFromLinearColorRGB(Color=&FNTimeOfDayColor(
      Night Color=<rgba 0.2614 0.3686 0.401 1>
      Morning Color=<rgba 0.6181 0.5423 0.8073 1>
      Day Color=<rgba 0.6517 0.919 1 1>
      Evening Color=<rgba 0.8073 0.5874 0.6207 1>
     ))
    ScaleA=true
    ScaleRGB=true
   )
   &FN_Mesh_Attributes(
    Color=&MultiplyLinearColor(
     Linear Color=<rgba 1 1 1 0.2>
     Scale Factor=&MakeVector4FromVectorAndFloat(
      "Vector (RGB)"=&MakeVectorFromLinearColorRGB(Color=&FNTimeOfDayColor(
        Night Color=<rgba 0.2614 0.3686 0.401 1>
        Morning Color=<rgba 0.6181 0.5423 0.8073 1>
        Day Color=<rgba 0.6517 0.919 1 1>
        Evening Color=<rgba 0.8073 0.5874 0.6207 1>
       ))
      "Float (Alpha)"=&FloatFromCurve(FloatCurve=DataInterface() {propertyValues {Curve {
          defaultValue=0
          keys [
           {arriveTangent=0.1155, interpMode=RCIM_Cubic, leaveTangent=0.1155, tangentMode=RCTM_User, time=0.4, value=1}
           {arriveTangent=-2.9743, interpMode=RCIM_Cubic, leaveTangent=-2.9743, tangentMode=RCTM_User, time=1, value=0}
          ]
         }}})
     )
    )
    Create Color Attribute=true
    Create Scale Attribute=true
    Scale=&Multiply_VectorByFloat(
     Vector=&VectorFromFloat(Value=&FloatFromCurve(
       FloatCurve=DataInterface() {propertyValues {Curve {
          defaultValue=0
          keys [
           {arriveTangent=1.6599, interpMode=RCIM_Cubic, leaveTangent=1.6599, tangentMode=RCTM_User, time=0, value=0.5}
           {interpMode=RCIM_Cubic, tangentMode=RCTM_User, time=1, value=1}
          ]
         }}}
       Scale Curve=&MakeFloatFromVector(Vector=$Particles.Initial.Scale, Channel=::NewEnumerator0)
      ))
     Float=1.3
    )
    Write to Module Namespace=true
   )
  ]
  rendererClasses [&NiagaraMeshRendererProperties]
  renderers [
   {rendererClass=&NiagaraMeshRendererProperties, rendererIndex=0}
   {rendererClass=&NiagaraMeshRendererProperties, rendererIndex=1}
  ]
 }
 SpeedLine001 {
  emitterUpdateScript [
   &EmitterState()
   &SpawnBurst_Instantaneous(Spawn Count=3, Use Spawn Probability=false, UseLoopCountLimit=false)
  ]
  particleSpawnScript [
   &InitializeParticle(
    Color=<rgba 1 1 1 10>
    Lifetime=0.25
    Sprite Rotation Mode=::NewEnumerator2
    Sprite Size Mode=::NewEnumerator1
    Uniform Sprite Size Max=1700
    Uniform Sprite Size Min=1020
    UsePositionOffset=false
    Write Material Random=true
   )
   &CameraOffset(Camera Offset Amount=200)
  ]
  particleUpdateScript [
   &ParticleState()
   &ScaleSpriteSize(
    Scale Sprite Size Mode=::NewEnumerator0
    Uniform Scale Factor=&FloatFromCurve(
     FloatCurve=DataInterface() {propertyValues {Curve {keys [{arriveTangent=3.9086, interpMode=RCIM_Cubic, leaveTangent=3.9086, tangentMode=RCTM_User, time=0, value=0}, <at 1 1.4 RCIM_Cubic>]}}}
     Scale Curve=&Lerp_Float(A=0.5, B=1, Alpha=$User.Size)
    )
   )
   &DynamicMaterialParameters(
    Index 0 Param 1=&FloatFromCurve(FloatCurve=DataInterface() {propertyValues {Curve {keys [<at 0 1 RCIM_Cubic>, {arriveTangent=-2.742, interpMode=RCIM_Cubic, leaveTangent=-2.742, tangentMode=RCTM_User, time=1, value=0}]}}})
    Module.Index0Float0Write=true
    Module.Index0Float1Write=true
    Module.Index0Float2Write=true
    Module.Index0Float3Write=true
   )
   &ScaleColor(
    Scale Alpha=&FloatFromCurve(
     FloatCurve=DataInterface() {propertyValues {Curve {keys [
         {arriveTangent=-13.9062, interpMode=RCIM_Cubic, leaveTangent=-13.9062, tangentMode=RCTM_User, time=0, value=7}
         {arriveTangent=-7, interpMode=RCIM_Cubic, leaveTangent=-7, time=0.2488, value=3}
         <at 1 0 RCIM_Cubic>
        ]}}}
     Scale Curve=2
    )
    ScaleA=true
    ScaleRGB=true
   )
  ]
  rendererClasses [&NiagaraSpriteRendererProperties]
  renderers [{rendererClass=&NiagaraSpriteRendererProperties}]
 }
}
summary {
 systemName=NS_BuzzCool_Ground_Impact
 emitters [
  {emitterName=Shockwaves_and_Decal, rendererClasses [&NiagaraMeshRendererProperties]}
  {emitterName=DarkStreaks, rendererClasses [&NiagaraSpriteRendererProperties]}
  {emitterName=SpeedLine001, rendererClasses [&NiagaraSpriteRendererProperties]}
  {emitterName=Smoke_Circular, rendererClasses [&NiagaraMeshRendererProperties]}
  {emitterName=Smoke_Masked002, rendererClasses [&NiagaraMeshRendererProperties]}
 ]
 userVariables [
  User.Frosting {type=LinearColor, description="", defaultValue {struct=LinearColor, value=<rgba 0.9062 0.0746 0.6196 1>}}
  User.Golden {type=NiagaraBool}
  User.Size {type=NiagaraFloat, defaultValue=1, description=""}
 ]
}
