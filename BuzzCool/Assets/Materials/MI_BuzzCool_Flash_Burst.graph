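# Unlit translucent material for Niagara sprites that draws a flash burst with animated, noise-distorted flare edges.
# Two counter-rotated panning noise samples, blended across U, offset the flare texture's UVs, weighted by a vertical gradient with a Mask offset parameter.
# The boosted flare, particle alpha and depth fade form opacity.
# A second, slower copy of the noise setup modulates brightness, multiplied by particle color.
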Material /BuzzCool/Assets/Materials/MI_BuzzCool_Flash_Burst
graph {
 properties {bUsedWithNiagaraSprites=true, blendMode=BLEND_Translucent, pixelDepthOffsetMode=PDOM_Legacy, shadingModel=MSM_Unlit, usedShadingModels=Unlit}
 nodes [
  n1 = ParticleColor()
  n2 = TextureCoordinate()
  n3 = Rotator(Coordinate=TextureCoordinate(), Time=Constant() {r=-3}) {speed=0.15}
  n4 = Constant2Vector() {g=0.5, r=2}
  n5 = TextureSample(UVs=Multiply(A=Panner(Coordinate=@n3) {speedY=2}, B=@n4)) {samplerType=SAMPLERTYPE_Grayscale, texture=&LowResBlurredNoise}
  n6 = Rotator(Coordinate=TextureCoordinate(), Time=Constant() {r=3}) {speed=0.15}
  n7 = TextureSample(UVs=Multiply(A=Panner(Coordinate=@n6) {speedY=2}, B=@n4)) {samplerType=SAMPLERTYPE_Grayscale, texture=&LowResBlurredNoise}
  n8 = ComponentMask(None=TextureCoordinate()) {r=true}
  n9 = Saturate(None=Multiply(A=SmoothStep(Min=@n8) {constValue=0.62}) {constB=3})
  n10 = TextureCoordinate()
  n11 = Add(A=OneMinus(None=ComponentMask(None=@n10) {g=true}), B=ScalarParameter() {parameterName=Mask offset})
  n12 = Multiply(A=Multiply(A=LinearInterpolate(A=@n5.R, B=@n7.R, Alpha=@n9)) {constB=0.1}, B=Saturate(None=Multiply(A=@n11) {constB=2}))
  n13 = Multiply(A=TextureSample(UVs=Add(A=@n2, B=@n12)) {texture=/Game/Effects/Fort_Effects/Emotes/DualParadox/T_Emote_DualParadox_Flare}.R) {constB=5}
  DepthFade(Opacity=Multiply(A=@n1.A, B=Saturate(None=@n13)), FadeDistance=ScalarParameter() {defaultValue=50, parameterName=FadeDistance}) {fadeDistanceDefault=50}
  n14 = Rotator(Coordinate=TextureCoordinate(), Time=Constant() {r=-3}) {speed=0.15}
  n15 = Constant2Vector() {g=0.5, r=2}
  n16 = TextureSample(UVs=Multiply(A=Panner(Coordinate=@n14) {speedY=1}, B=@n15)) {samplerType=SAMPLERTYPE_Grayscale, texture=&LowResBlurredNoise}
  n17 = Rotator(Coordinate=TextureCoordinate(), Time=Constant() {r=3}) {speed=0.15}
  n18 = TextureSample(UVs=Multiply(A=Panner(Coordinate=@n17) {speedY=1}, B=@n15)) {samplerType=SAMPLERTYPE_Grayscale, texture=&LowResBlurredNoise}
  n19 = ComponentMask(None=TextureCoordinate()) {r=true}
  n20 = Saturate(None=Multiply(A=SmoothStep(Min=@n19) {constValue=0.62}) {constB=3})
  Multiply(A=LinearInterpolate(Alpha=LinearInterpolate(A=@n16.R, B=@n18.R, Alpha=@n20)) {constA=-0.1, constB=3}, B=@n1.RGB)
 ]
}
