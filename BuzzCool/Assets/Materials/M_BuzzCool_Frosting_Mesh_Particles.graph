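# Masked, two-sided lit material for mesh particles, including Niagara, named for BuzzCool frosting, with a dissolve and a two-tone pink color.
# An optionally inverted cloud texture plus particle alpha is shaped by power and multiplier parameters, then fed to a temporal-AA dither function for cutout.
# Camera-to-normal dot product blends low and high colors, while low quality and mobile feature levels use the high color alone.
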Material /BuzzCool/Assets/Materials/M_BuzzCool_Frosting_Mesh_Particles
graph {
 properties {bUsedWithMeshParticles=true, bUsedWithNiagaraMeshParticles=true, blendMode=BLEND_Masked, twoSided=true, usedShadingModels=Default Lit}
 nodes [
  Constant() {r=1}
  Constant()
  n1 = TextureSampleParameter2D(UVs=Multiply(A=TextureCoordinate(), B=ScalarParameter() {defaultValue=1, parameterName=Dissolve Scale})) {parameterName=Dissolve Texture, texture=/Game/Effects/Fort_Effects/Textures/Smoke/T_Clouds_01}
  n2 = OneMinus(None=@n1.R)
  n3 = ScalarParameter() {parameterName=Invert Dissolve}
  n4 = Saturate(None=Add(A=LinearInterpolate(A=@n1.R, B=@n2, Alpha=@n3), B=ParticleColor().A))
  n5 = ScalarParameter() {defaultValue=2, parameterName=Dissolve Power}
  MaterialFunctionCall(Alpha Threshold=Multiply(A=Power(Base=@n4, Exp=@n5), B=ScalarParameter() {defaultValue=3.5, parameterName=Dissolve Multi})) {materialFunction=&DitherTemporalAA_Fortnite}
  n6 = VectorParameter() {group=Color, parameterName=BaseColorHigh, defaultValue=<rgba 0.7917 0.0948 0.5337 1>}
  n7 = LinearInterpolate(
   A=VectorParameter() {group=Color, parameterName=BaseColorLow, defaultValue=<rgba 0.7991 0.0666 0.5583 1>}.RGB
   B=@n6.RGB
   Alpha=DotProduct(A=CameraVectorWS(), B=VertexNormalWS())
  )
  n8 = FeatureLevelSwitch(Default=@n7, ES2=@n6.RGB, ES3_1=@n6.RGB, SM4_REMOVED=@n7)
  n9 = QualitySwitch(Default=@n8, Low=@n6.RGB, High=@n8, Medium=@n8)
  Multiply(A=@n9, B=ScalarParameter() {defaultValue=0.4, parameterName=B}) {constB=0.5}
  Reroute(None=@n9)
 ]
}
