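# Deferred decal material for BuzzCool that draws a hard-edged, noisy frosting splat in a tintable color, pink by default.
# A radial UV falloff, broken up by noise with position-varied tiling, is rounded into the color and opacity mask and shrinks with decal lifetime.
# A decal-facing mask limits opacity.
# Noise normals, flattened by the falloff, blend over scene normals (constant on some platforms) and feed the Buzzcool color-banding function.
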Material /BuzzCool/Assets/Materials/M_Buzzcool_Decal
graph {
 properties {
  bAutomaticallySetUsageInEditor=false
  bUseMaterialAttributes=true
  blendMode=BLEND_Translucent
  materialDomain=MD_DeferredDecal
  usedShadingModels=Default Lit
 }
 nodes [
  n1 = Max(A=DecalLifetimeOpacity().Opacity) {constB=0.001}
  n2 = ScalarParameter() {defaultValue=1, parameterName=Frosting Distance Min}
  n3 = ScalarParameter() {defaultValue=0.85, parameterName=Frosting Distance Max}
  n4 = TextureCoordinate()
  n5 = Length(None=ConstantBiasScale(None=@n4) {bias=-0.5, scale=2}) {is=class MaterialExpressionLength}
  n6 = Saturate(None=InverseLinearInterpolate(Value=@n5) {constA=1, constB=0})
  n7 = VectorParameter() {parameterName=Frosting Noise Tiling, defaultValue=<rgba 3 2 0 1>}
  n8 = MaterialFunctionCall(Float3=ActorPositionWS()) {materialFunction=&BreakOutFloat3Components}
  n9 = Add(A=@n8.R, B=@n8.R)
  n10 = LinearInterpolate(Alpha=Frac(None=Add(A=@n9, B=@n8.B))) {constA=-1, constB=1.5}
  n11 = Multiply(A=@n4, B=ComponentMask(None=Add(A=@n7.RGB, B=@n10)) {g=true, r=true})
  n12 = Multiply(
   A=InverseLinearInterpolate(A=Multiply(A=@n1, B=@n2), B=Multiply(A=@n1, B=@n3), Value=@n5)
   B=LinearInterpolate(
    A=Power(Base=@n6)
    Alpha=TextureSample(UVs=@n11) {samplerType=SAMPLERTYPE_Grayscale, texture=/Game/Global/Textures/Default/Noise/LowResBlurredNoiseNormalized}.R
   )
  )
  n13 = Round(None=Saturate(None=@n12))
  n14 = SetMaterialAttributes!(
   MaterialAttributes
   Base Color=LinearInterpolate(A=Constant3Vector(), B=VectorParameter() {parameterName=Frosting Color, defaultValue=<rgba 0.8031 0.0209 0.5188 0>}.RGB, Alpha=@n13)
   Metallic=Constant()
   Specular=Constant() {r=0.5}
   Roughness=Constant() {r=0.5}
   Opacity=Multiply(A=Max(B=@n13), B=MaterialFunctionCall() {materialFunction=/Game/Weapons/Prototype/SprayGun/Materials/MF_DecalFacingMask}.Result) {constA=1}
  ) {attributeSetTypes [
    #1
    #2
    #3
    #4
    #5
   ]}
  n15 = TextureSample(UVs=@n11) {samplerType=SAMPLERTYPE_Normal, texture=/BuzzCool/Assets/Materials/Textures/LowResBlurredNoiseNormal}
  n16 = MaterialFunctionCall(X=@n15.G, Y=Multiply(A=@n15.R) {constB=-1}, Z=@n15.G) {materialFunction=&MakeFloat3}
  n17 = Constant3Vector() {constant=<rgba 0 0 1 0>}
  n18 = InverseLinearInterpolate(B=ScalarParameter() {defaultValue=3, parameterName=Frosting Normal Range}, Value=@n12) {constA=0.5}
  n19 = Saturate(None=@n18)
  n20 = Saturate(None=Multiply(A=@n18) {constB=-1})
  n21 = Transform(None=LinearInterpolate(A=LinearInterpolate(A=@n16.Result, B=@n17, Alpha=@n19), B=@n17, Alpha=Ceil(None=@n20)))
  MaterialFunctionCall!(
   In=@n14
   Banding Mask=@n13
   NormalWS=MaterialFunctionCall(BaseNormal=NamedRerouteUsage(), AdditionalNormal=Normalize(VectorInput=@n21)) {materialFunction=&BlendAngleCorrectedNormals}.Result
  ) {materialFunction=&/MF_Buzzcool_ColorBanding}
  n22 = SceneTexture() {sceneTextureId=PPI_WorldNormal}
  n23 = Constant3Vector() {constant=<rgba 0 0 1 0>}
  NamedRerouteDeclaration(None=MaterialFunctionCall!("Xb1,PS4"=@n22.Color, "PC-High&&Epic,XSX,PS5"=@n22.Color, "PC-Perf,Android"=@n23, Switch=@n23, "PC-Low(NotPerf)&&Medium"=@n22.Color) {materialFunction=/Game/Effects/Niagara/MasterMaterials/Functions/Common/PlatformSwitch_3D}.Result) {name=NormalWS, nodeColor=<rgba 0.4766 1 0 1>}
 ]
}
