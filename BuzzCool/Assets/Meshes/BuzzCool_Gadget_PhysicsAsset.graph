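# Physics asset for the BuzzCool gadget mesh, giving it simple collision as a single rigid body.
# One capsule on the root bone, slightly raised from the bone origin and rotated, stands in for the whole shape.
# It has no constraints and no other bodies, so it defines no articulated or ragdoll motion between parts.
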PhysicsAsset /BuzzCool/Assets/Meshes/BuzzCool_Gadget_PhysicsAsset
bodies {root [{shapeType=Capsule, length=7.8768, radius=14.2485, center=<-0.0292 -0.0154 5.4876>, rotation=<rot 0 3.6448 -89.9715>}]}
constraints []
